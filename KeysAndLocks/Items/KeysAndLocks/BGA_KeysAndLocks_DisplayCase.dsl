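; Keys-and-locks display case that shows a holographic weapon and drops it when unlocked.
; Server rolls loot and mod set once loot tables apply, re-rolling on Creative Loot Controller events, and replicates the item definition.
; Clients async-load rarity-tinted hologram meshes, with timelines animating show/hide and opening.
; Lock prompts retry on a timer because child actors replicate without callbacks; opening sends a quest event and toggles net dormancy.
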(event EventBeginPlay
  (bind _returnvalue (Fortnite|Athena|GameState|GetGameStateAthena))
  (bind _itemmesh_skeletal (Variables|Default|GetItemMesh(Skeletal)))
  ;[0 1 0.07604] BeginPlay
  (Utilities|FlowControl|Sequence
    (:then_0
      (Networking|SwitchHasAuthority
        (:Authority
          ;;[0.165 0.2109 1] Set the owner of each lock to this. Also have each lock use its owner for relevancy checks.
          (for _array_element (CallFunction|GetLockChildActors)
            (Utilities|IsValid _array_element
              (:Is%20Valid
                (Utilities|IsValid (Variables|/Script/Engine.ChildActorComponent|GetChildActor _array_element)
                  (:Is%20Valid
                    (Variables|/Script/Engine.Actor|SetbNetUseOwnerRelevancy true (Variables|/Script/Engine.ChildActorComponent|GetChildActor _array_element))
                    (Actor|SetOwner (Variables|/Script/Engine.ChildActorComponent|GetChildActor _array_element) self)))))))
        (:Remote
          (Variables|/Script/Engine.SkeletalMeshComponent|SetbNoSkeletonUpdate true _itemmesh_skeletal)
          (Rendering|SetVisibility _itemmesh_skeletal false false)
          (CallFunction|UpdateItemDefinition)
          (if (Variables|Loot|GetOpened)
            (Rendering|Material|SetScalarParameterValueonMaterials (Variables|Default|GetChest) "Code_Timestamp" 0.225)
            (bind _valid_scene_component (Utilities|ForEachValidSceneComponent (Variables|Hologram|GetHologram%20Scene%20Components))
              (:Valid%20Loop%20Body
                (Rendering|SetVisibility _valid_scene_component false false)))))))
    (:then_1
      (if (Variables|DisplayItemMesh|GetUseCustomStaticMesh)
        (Variables|DisplayItemMesh|SetCustomStaticMeshScale (Variables|/Script/Engine.SceneComponent|GetRelativeScale3D (Variables|Default|GetItemMesh(Static))))))
    (:then_2
      ;;[1 0.3891 0] For Creative Game Modes: The Loot Controller Device will broadcast out when a new loot pool has been selected, display cases need to listen for this broadcast and update their loot accordingly.
      (Networking|SwitchHasAuthority
        (:Authority
          (Utilities|IsValid _returnvalue
            (:Is%20Valid
              (if (Class|FortGameStateAthena|IsSpatialGameplayExperience _returnvalue)
                (bind _returnvalue_2 (GameplayEventRouter|GetEventRouter "/Script/FortniteGame.FortPlayspace" self))
                (Utilities|IsValid _returnvalue_2
                  (:Is%20Valid
                    (bind (_asynctaskproxy _payload) (GameplayEventRouter|StartListeningToEvent _returnvalue_2 "/Game/Creative/Devices/LootController/Events/UpdateDisplayCaseLootEvent.UpdateDisplayCaseLootEvent")
                      (:then
                        (Variables|Default|SetUpdateDisplayCaseLootAsyncTask _asynctaskproxy))
                      (:OnEventReceived
                        (bind _payload_shoulddropammo_9_f2d8e01b4aa2cf60dbc21db92846b560 (Utilities|Struct|BreakUpdateDisplayCaseLootEvent _payload))
                        (Variables|Ammo|SetSpawnAmmo _payload_shoulddropammo_9_f2d8e01b4aa2cf60dbc21db92846b560)
                        (Utilities|Array|Clear (Variables|/KeysAndLocks/Items/KeysAndLocks/BGA_KeysAndLocks_DisplayCasesManager.BGA_KeysAndLocks_DisplayCasesManager_C|GetGeneratedItems (Variables|Default|GetSpawner)))
                        (CallFunction|ChooseandInitializeItemDefinition)))))))))))))

(event EventEndPlay
  (params (EndPlayReason :type EEndPlayReason%20Enum :default "Destroyed"))
  ;[1 0.5326 0] EndPlay
  (|CallParentFunction EndPlayReason)
  (Utilities|Time|ClearandInvalidateTimerbyHandle (Variables|Default|GetUpdateLockChildActorsTimerHandle))
  (CallFunction|UnbindLootTablesApplied))

(event Custom|Init
  (params (Spawner :type BGA%20Keys%20And%20Locks%20Display%20Cases%20Manager%20Object%20Reference))
  ;[0 1 0.3871] Init
  (Variables|Default|SetSpawner Spawner)
  (CallFunction|Log "Init" "called")
  (Utilities|FlowControl|Sequence
    (:then_0
      (LinkToActorComponent|PerformLinkingCheck (Variables|Default|GetFortLinkToActor) "Down" (Variables|Default|GetLinkedActor)))
    (:then_1
      (Networking|SwitchHasAuthority
        (:Authority
          (bind _asfort_game_mode_athena (Utilities|Casting|CastToFortGameModeAthena (Game|GetGameMode)))
          (if (Class|FortGameModeAthena|HavePlaylistLootTablesBeenApplied _asfort_game_mode_athena)
            (CallFunction|Log "Init" "Loot is ready, Calling ChooseAndInitializeItemDefinition")
            (CallFunction|ChooseandInitializeItemDefinition)
            (else
              (CallFunction|Log "Init" "Loot is NOT ready, binding to OnPlaylistLootTablesAppliedDelegate")
              (EventDispatchers|BindEventtoOnPlaylistLootTablesAppliedDelegate _asfort_game_mode_athena (EventDispatchers|CreateEvent :!event Custom|ChooseAndInitializeItemDefinition)))))))))

(event Custom|ChooseAndInitializeItemDefinition
  ;[0 1 0.5812] ChooseAndInitializeItemDefinition
  (Networking|SwitchHasAuthority
    (:Authority
      (bind (_asfort_weapon_item_definition _loot) (CallFunction|GetRandomRareorLegendaryWeapon))
      (CallFunction|Log "ChooseAndInitializeItemDefinition" (Utilities|String|Append "Item Definition: [" (Utilities|String|ToString(Text) (ItemDefinition|GetDisplayName _asfort_weapon_item_definition)) "]"))
      (Utilities|Array|Add (Variables|/KeysAndLocks/Items/KeysAndLocks/BGA_KeysAndLocks_DisplayCasesManager.BGA_KeysAndLocks_DisplayCasesManager_C|GetGeneratedItems (Variables|Default|GetSpawner)) (Weapon|GetUpgradedWeaponItemVerticaltoRarity _asfort_weapon_item_definition "Legendary"))
      (Variables|Default|SetItemDefinition _asfort_weapon_item_definition)
      (bind _output_get (Variables|Loot|SetLoot _loot))
      (Networking|ForceNetUpdate)
      (CallFunction|UnbindLootTablesApplied)
      (if (<= (Utilities|Array|Length _output_get) 0)
        ;;[1 0.05258 0] Failed to find an item for this cheat, it will be empty. Spit out a warning about how to fix this.
        (Fortnite|FortShippingLog self (Utilities|String|Append "BGA_KeysAndLocks_DisplayCase::ChooseAndInitializeItemDefinition() [" (Utilities|GetObjectName self) "] failed to set an item and will appear empty." " To solve this you should either 1) change \"Tier Group Name\" to a tier group that currently enabled and has enabled loot. 2) prevent this display case from spawning by disabling or modifying the BP_KeysAndLocks_SpawnerConfig... that could spawn this chest (use the reference viewer). You can likely disable that spawnerconfig by modifying KeysAndLocksGameData.") true)))))

(event Custom|Spawn%20Loot%20%28Server%20Only%29
  (bind _lockchildactor (Variables|Default|GetLockChildActor))
  (bind _returnvalue (Transformation|GetWorldLocation _lockchildactor))
  (bind _returnvalue_2 (Transformation|GetWorldRotation _lockchildactor))
  (bind _returnvalue_3 (Math|Rotator|GetRotationXVector _returnvalue_2))
  (bind _returnvalue_4 (* _returnvalue_3 3e+01))
  (bind _returnvalue_5 (Utilities|Operators|Add _returnvalue "0,0,65.000000" _returnvalue_4))
  (bind _weponmodsetdata (Variables|ModSetData|GetWeponModSetData))
  ;[0.2314 0.1848 1] Spawn Loot (Server Only)
  (Networking|SwitchHasAuthority
    (:Authority
      (bind _itempickupsspawned (Class|AthenaFunctionLibrary|SpawnLootfromArray (Variables|Loot|GetLoot) _returnvalue_5 _returnvalue_3 0 true false false 0 "Other" "Chest"))
      (Utilities|IsValid _weponmodsetdata
        (:Is%20Valid
          (for _array_element _itempickupsspawned
            (Class|FortWeaponModFunctionLibrary|ApplyModSettoPickup _weponmodsetdata _array_element)
            (if (Variables|Ammo|GetSpawnAmmo)
              (bind _asfort_weapon_item_definition (Utilities|Casting|CastToFortWeaponItemDefinition (Variables|Default|GetItemDefinition)))
              (Fortnite|Weapon|CreateTossAmmoPickupforWeaponItemDefinitionatLocation _asfort_weapon_item_definition (GameplayTags|MakeLiteralGameplayTagContainer (Variables|ammo|GetDisplayCaseAmmoTag)) _returnvalue_5))))
        (:Is%20Not%20Valid
          _)))))

(event OpenableInterface|EventOpen
  (params (ControllerInstigator :type Controller%20Object%20Reference) (bRequestFastOpen :type Boolean))
  ;[0.6262 0.3719 1] Event Open
  (if (not (Variables|Loot|GetOpened))
    (CallFunction|SendOpenQuestEvent)
    (Variables|Loot|SetOpened true)
    (Networking|SetNetDormancy "DORM_Awake") ; Wake up for the duration of the replicated timeline
    (bind (_direction _openprogress) (Timeline|OpenAnim PlayFromStart)
      (:Update
        ;;[0 0.886 1] OpenAnim Update
        (Networking|CanExecuteCosmeticEvents
          (:True
            (Rendering|Material|SetScalarParameterValueonMaterials (Variables|Default|GetChest) "Code_Timestamp" _openprogress))))
      (:Finished
        ;;[1 0.6284 0.165] OpenAnim Finished
        (Networking|SwitchHasAuthority
          (:Authority
            (bind _valid_scene_component (Utilities|ForEachValidSceneComponent (Variables|Hologram|GetHologram%20Scene%20Components))
              (:Valid%20Loop%20Body
                (Components|DestroyComponent _valid_scene_component))
              (:Completed
                (Networking|FlushNetDormancy)
                (Networking|SetNetDormancy "DORM_DormantAll"))))))
      (:SpawnLoot
        ;;[0.0582 0.05294 1] OpenAnim Spawn Loot
        (Networking|SwitchHasAuthority
          (:Authority
            (CallFunction|SpawnLoot(ServerOnly))
            (Variables|Loot|Setb_LootSpawned true)
            (Default|CallOnLockedChestOpenedServer) ; Use for Rift Encounters or anybody else that wants to listen to this event
          )
          (:Remote
            (Audio|PlaySoundatLocation "/Game/Sounds/Foley_Loot/Containers/Keys_And_Locks/Keys_And_Locks_Open_Cue.Keys_And_Locks_Open_Cue" (Transformation|GetActorLocation self)) ; Case Open SFX
            (Ability|RemoveGameplayCueReplicated (Variables|GameplayCues|GetGC_HologramSound))
            (bind _valid_scene_component_2 (Utilities|ForEachValidSceneComponent (Variables|Hologram|GetHologram%20Scene%20Components))
              (:Valid%20Loop%20Body
                (Development|SetHiddeninGame _valid_scene_component_2 true false)))))))))

(event Custom|LoadMesh
  (bind _itemdefinition (Variables|Default|GetItemDefinition))
  (bind _itemdefinition_2 (Variables|Default|GetItemDefinition))
  (bind _weponmodsetdata (Variables|ModSetData|GetWeponModSetData))
  (if (not (Variables|Default|GetAlreadyLoaded))
    (Variables|Default|SetAlreadyLoaded true)
    (if (Variables|DisplayItemMesh|GetUseCustomStaticMesh)
      (Components|DestroyComponent (Variables|Default|GetItemMesh(Skeletal)))
      (else
        (bind _itemdefinition_3 (Variables|Default|GetItemDefinition)
          (:then
            (bind _itemdata (Class|ItemDefinitionBase|GetItemData _itemdefinition_3)
              (:Found
                (CallFunction|Log "LoadMesh" "Item Definition is valid.")
                (bind _output_get (Variables|DisplayItemMesh|SetUsingSkeletal (Utilities|IsValidSoftObjectReference (.PickupSkeletalMesh (Utilities|Struct|BreakFortnitePickupItemData _itemdata)))))
                (if _output_get
                  (bind _loaded_object (Creative|Loading|AsyncLoadEditorSafe (.PickupSkeletalMesh (Utilities|Struct|BreakFortnitePickupItemData _itemdata)))
                    (:Loaded
                      (Components|DestroyComponent (Variables|Default|GetItemMesh(Static)))
                      (Components|SkinnedMesh|SetSkinnedAssetandUpdate (Variables|Default|GetItemMesh(Skeletal)) (Utilities|Casting|CastToSkeletalMesh _loaded_object) true)
                      (CallFunction|CenterSkeletalMesh)
                      (Utilities|FlowControl|Sequence
                        (:then_0
                          (if (== (Variables|Default|GetItemDefinition) "/Game/Athena/Items/Consumables/ShockwaveGrenade/Athena_ShockGrenade.Athena_ShockGrenade")
                            (Transformation|SetWorldScale3D (Variables|Default|GetItemMesh(Skeletal)) "2.000000,2.000000,2.000000")))
                        (:then_1
                          ;[1 1 1] Set rarity color on mesh
                          (Networking|SwitchHasAuthority
                            (:Remote
                              (Utilities|IsValid _itemdefinition
                                (:Is%20Valid
                                  (CallFunction|SetProjectionRarityColor (Item|GetRarity _itemdefinition) (Variables|Default|GetItemMesh(Skeletal)))
                                  ;[1 1 1] Load Weapon Mods
                                  (if (Utilities|IsValid _weponmodsetdata :!class /Script/Engine.KismetSystemLibrary)
                                    (bind _asfort_weapon_item_definition (Utilities|Casting|CastToFortWeaponItemDefinition (Variables|Default|GetItemDefinition)))
                                    (bind _returnvalue (Class|FortWeaponModFunctionLibrary|GetModDisplayData _asfort_weapon_item_definition _weponmodsetdata))
                                    (for _array_element _returnvalue
                                      (switch Utilities|FlowControl|Switch|SwitchonGameplayTag (Utilities|Struct|BreakFortModDisplayData _array_element)
                                        (:WeaponMod.Slot.Underbarrel
                                          (Variables|Default|SetModBoneName "attach_under")
                                          (Variables|Default|SetModBoneOffset "0.000000,0.000000,-15.000000")
                                          (CallFunction|SetMeshandRarityColor (Variables|Default|GetCaseModMesh3) (.ModMesh (Utilities|Struct|BreakFortModDisplayData _array_element))))
                                        (:WeaponMod.Slot.Barrel
                                          (Variables|Default|SetModBoneName "attach_side")
                                          (Variables|Default|SetModBoneOffset "0.000000,20.000000,10.000000")
                                          (CallFunction|SetMeshandRarityColor (Variables|Default|GetCaseModMesh2) (.ModMesh (Utilities|Struct|BreakFortModDisplayData _array_element))))
                                        (:WeaponMod.Slot.Optic
                                          (Variables|Default|SetModBoneName "attach_optic")
                                          (Variables|Default|SetModBoneOffset "0.000000,0.000000,15.000000")
                                          (CallFunction|SetMeshandRarityColor (Variables|Default|GetCaseModMesh1) (.ModMesh (Utilities|Struct|BreakFortModDisplayData _array_element))))
                                        (:WeaponMod.Slot.Magazine
                                          (Variables|Default|SetModBoneName "attach_mag")
                                          (Variables|Default|SetModBoneOffset "0.000000,0.000000,-15.000000")
                                          (CallFunction|SetMeshandRarityColor (Variables|Default|GetCaseModMesh4) (.ModMesh (Utilities|Struct|BreakFortModDisplayData _array_element)))))))))))))))
                  (else
                    (bind _loaded_object_2 (Creative|Loading|AsyncLoadEditorSafe (Utilities|Struct|BreakFortnitePickupItemData _itemdata))
                      (:Loaded
                        (Components|DestroyComponent (Variables|Default|GetItemMesh(Skeletal)))
                        (Components|StaticMesh|SetStaticMesh (Variables|Default|GetItemMesh(Static)) (Utilities|Casting|CastToStaticMesh _loaded_object_2))
                        (Utilities|IsValid _itemdefinition_2
                          (:Is%20Valid
                            (CallFunction|SetProjectionRarityColor (Item|GetRarity _itemdefinition_2) (Variables|Default|GetItemMesh(Static))))))))))
              (:NotFound
                _)))
          (:else
            _))
        (CallFunction|Log "LoadMesh" "Failed, Item Definition == nullptr.")))))

(event Custom|Show%20Weapon
  (bind _returnvalue (Ability|GameplayCue|MakeGameplayCueParameters 0.0))
  (bind (_r _g _b) (Math|Color|BreakColor (Variables|Colors|GetBeamColor)))
  (bind (_r_2 _g_2 _b_2) (Math|Color|BreakColor (Variables|Colors|GetBeamColor)))
  (bind _usecustomstaticmesh (Variables|DisplayItemMesh|GetUseCustomStaticMesh))
  (bind _returnvalue_2 (not _usecustomstaticmesh))
  (bind _usingskeletal (Variables|DisplayItemMesh|GetUsingSkeletal))
  (bind _returnvalue_3 (and _usingskeletal _returnvalue_2))
  (bind _itemmesh_skeletal (Variables|Default|GetItemMesh(Skeletal)))
  (bind _itemmesh_static (Variables|Default|GetItemMesh(Static)))
  (bind _returnvalue_4 (select _returnvalue_3 _itemmesh_skeletal _itemmesh_static))
  (bind _itemmesh_skeletal_2 (Variables|Default|GetItemMesh(Skeletal)))
  (bind (_relativelocation_x _relativelocation_y) (Math|Vector|BreakVector (Variables|/Script/Engine.SceneComponent|GetRelativeLocation_X _returnvalue_4)))
  (bind _itemmesh_skeletal_3 (Variables|Default|GetItemMesh(Skeletal)))
  ;[1 0 0.9262 0.469] Show/Hide Weapon
  (Networking|CanExecuteCosmeticEvents
    (:True
      (if (not (Variables|Loot|GetOpened))
        (GameplayCue|ExecuteGameplayCue%28Burst%29Local self (Variables|GameplayCues|GetGC_HologramPopOut) _returnvalue)
        (Projectile|AddGameplayCue%28Looping%29Local self (Variables|GameplayCues|GetGC_HologramSound) _returnvalue)
        (Rendering|SetVisibility _itemmesh_skeletal_2 true false)
        (Variables|/Script/Engine.SkeletalMeshComponent|SetbNoSkeletonUpdate false _itemmesh_skeletal_2)
        (bind (_direction _beamopacity _scale) (Timeline|AppearTimeline Play)
          (:Update
            ;;[0.02034 0.8602 1] AppearTimeline Update
            (Utilities|FlowControl|Sequence
              (:then_0
                (bind _ns_keyslocks_holo_beam (Variables|Default|GetNS_KeysLocks_Holo_Beam)
                  (:then
                    (Effects|Components|ParticleSystem|SetColorParameter _ns_keyslocks_holo_beam "Beam Color" (Math|Color|MakeColor _r_2 _g_2 _b_2 (* _beamopacity 2.0)))
                    (Niagara|SetNiagaraVariable(Float) _ns_keyslocks_holo_beam "EntrySignal" (* _beamopacity 2.0)))
                  (:else)))
              (:then_1
                (Utilities|IsValid _returnvalue_4
                  (:Is%20Valid
                    (Transformation|SetRelativeScale3D _returnvalue_4 (* (Math|Vector|SelectVector (Variables|DisplayItemMesh|GetCustomStaticMeshScale) "1.000000,1.000000,1.000000" (Variables|DisplayItemMesh|GetUseCustomStaticMesh)) _scale))
                    (Transformation|SetRelativeLocation _returnvalue_4 (Math|Vector|MakeVector _relativelocation_x _relativelocation_y (Utilities|RemapFloat _scale "(X=0.000000,Y=1.000000)" "(X=60.000000,Y=125.000000)")) false false))))))
          (:Finished
            ;;[1 0.5237 0.03349] AppearTimeline Finished
            (if (Utilities|Enum|Equal(Enum) _direction "Backward")
              (Variables|/Script/Engine.SkeletalMeshComponent|SetbNoSkeletonUpdate true _itemmesh_skeletal_3)
              (bind _valid_scene_component (Utilities|ForEachValidSceneComponent (Utilities|Array|MakeArray (Variables|Default|GetItemMesh(Static)) _itemmesh_skeletal_3))
                (:Valid%20Loop%20Body
                  (Rendering|SetVisibility _valid_scene_component false false))
                (:Completed
                  (bind _ns_keyslocks_holo_beam_2 (Variables|Default|GetNS_KeysLocks_Holo_Beam)
                    (:then
                      (Effects|Components|ParticleSystem|SetColorParameter _ns_keyslocks_holo_beam_2 "Beam Color" (Math|Color|MakeColor _r _g _b (Variables|Hologram|GetBeamOpacityWhenPlayerTooFar))))
                    (:else)))))))))))

(event Custom|Hide%20Weapon
  (bind _returnvalue (Ability|GameplayCue|MakeGameplayCueParameters 0.0))
  (bind (_r _g _b) (Math|Color|BreakColor (Variables|Colors|GetBeamColor)))
  (bind (_r_2 _g_2 _b_2) (Math|Color|BreakColor (Variables|Colors|GetBeamColor)))
  (bind _usecustomstaticmesh (Variables|DisplayItemMesh|GetUseCustomStaticMesh))
  (bind _returnvalue_2 (not _usecustomstaticmesh))
  (bind _usingskeletal (Variables|DisplayItemMesh|GetUsingSkeletal))
  (bind _returnvalue_3 (and _usingskeletal _returnvalue_2))
  (bind _itemmesh_skeletal (Variables|Default|GetItemMesh(Skeletal)))
  (bind _itemmesh_static (Variables|Default|GetItemMesh(Static)))
  (bind _returnvalue_4 (select _returnvalue_3 _itemmesh_skeletal _itemmesh_static))
  (bind (_relativelocation_x _relativelocation_y) (Math|Vector|BreakVector (Variables|/Script/Engine.SceneComponent|GetRelativeLocation_X _returnvalue_4)))
  (bind _itemmesh_skeletal_2 (Variables|Default|GetItemMesh(Skeletal)))
  ;[1 0 0.9262 0.469] Show/Hide Weapon
  (Networking|CanExecuteCosmeticEvents
    (:True
      (GameplayCue|ExecuteGameplayCue%28Burst%29Local self (Variables|GameplayCues|GetGC_HologramPopIn) _returnvalue)
      (Projectile|RemoveGameplayCue%28Looping%29Local self (Variables|GameplayCues|GetGC_HologramSound) _returnvalue)
      (Timeline|AppearTimeline Reverse))))

(event Custom|UnbindLootTablesApplied
  ;[1 0.6091 0] UnbindLootTablesApplied
  (bind _asfort_game_mode_athena (Utilities|Casting|CastToFortGameModeAthena (Game|GetGameMode)))
  (EventDispatchers|UnbindEventfromOnPlaylistLootTablesAppliedDelegate _asfort_game_mode_athena (EventDispatchers|CreateEvent :!event Custom|ChooseAndInitializeItemDefinition)))

(fn BlueprintCanInteract
  (params (InteractingPawn :type Fort%20Pawn%20Object%20Reference) (InteractionBeingAttempted :type EInteractionBeingAttempted%20Enum :default "FirstInteraction") (InteractionType :type TInteractionType%20Enum :default "IT_NoInteraction"))
  (return (and (Variables|Loot|GetOpened) (not (Variables|Loot|GetItemTaken)))))

(fn BlueprintGetInteractionString
  (params (InteractingPawn :type Fort%20Pawn%20Object%20Reference) (InteractionBeingAttempted :type EInteractionBeingAttempted%20Enum :default "FirstInteraction"))
  (return (select (Variables|Loot|Getb_LootSpawned) "NSLOCTEXT(\"[62AC31B24745139BE344D2B15B75F839]\", \"FBD3472D4FB6EADDAE985D80C0ECC51A\", \"Take Item\")" "NSLOCTEXT(\"[62AC31B24745139BE344D2B15B75F839]\", \"283EAB3246B1E2DBF2C675B843ECD2E8\", \"Unlock (Uses 1 Key)\")")))

(fn CalculateLockInteractStrings
  (params (Lock%20Child%20Actors :type Array%20of%20Child%20Actor%20Component%20Object%20References))
  (bind _showcount (Variables|Text|GetShowCount))
  (bind _loot (Variables|Loot|GetLoot))
  (bind (_output_itemdefinition _output_itemguid _output_count) (Class|FortItemFunctionLibrary|BreakFortItemEntry (Utilities|Array|Get(aref) _loot)))
  (bind _returnvalue (> _output_count 1))
  (bind _returnvalue_2 (and _returnvalue _showcount))
  (bind _returnvalue_3 (Utilities|String|ToString(Integer) _output_count))
  (bind _returnvalue_4 (Utilities|String|Append "x" _returnvalue_3))
  (bind _returnvalue_5 (Utilities|String|SelectString _returnvalue_4 "" _returnvalue_2))
  (bind _weponmodsetdata (Variables|ModSetData|GetWeponModSetData))
  (bind _returnvalue_6 (Utilities|IsValid _weponmodsetdata :!class /Script/Engine.KismetSystemLibrary))
  (bind _modsetname (Variables|/Script/FortniteGame.FortWeaponModSetData|GetModSetName _weponmodsetdata))
  ;[1 1 1] Nativized in Native Calculate Lock Interact Strings for FORT-636113
  (if (Utilities|Array|IsValidIndex (Variables|Loot|GetLoot))
    (bind _itemdefinition (Variables|Default|GetItemDefinition)
      (:then
        (bind _result (ArrayMacros|SafeArrayGet (Variables|Text|GetRarityTexts) (Math|Conversions|ToInteger(Byte) (Item|GetRarity _itemdefinition)))
          (:%20%20
            (for _array_element Lock%20Child%20Actors
              (Utilities|IsValid _array_element
                (:Is%20Valid
                  (Utilities|IsValid (Variables|/Script/Engine.ChildActorComponent|GetChildActor _array_element)
                    (:Is%20Valid
                      (bind _asbga_keys_and_locks_lock (Utilities|Casting|CastToBGA_KeysAndLocks_Lock (Variables|/Script/Engine.ChildActorComponent|GetChildActor _array_element)))
                      (Variables|/KeysAndLocks/Items/KeysAndLocks/BGA_KeysAndLocks_Lock.BGA_KeysAndLocks_Lock_C|SetInteractionText_HaveKey (Utilities|Text|FormatText "NSLOCTEXT(\"[62AC31B24745139BE344D2B15B75F839]\", \"87B10A9A4ED9081CA4D5538175BC4EC4\", \"Use Key\\r\\n{1}\")" (Utilities|Text|FormatText "INVTEXT(\"{0}\\r\\n{1} {2}\")" (select _returnvalue_6 (Utilities|Text|FormatText "NSLOCTEXT(\"[62AC31B24745139BE344D2B15B75F839]\", \"5C85BE1B4C0B44D72BBF689B74C0B055\", \"{ModSetName}\\r\\n{WeaponName}\")" _modsetname (ItemDefinition|GetDisplayName _itemdefinition)) (ItemDefinition|GetDisplayName _itemdefinition)) _result _returnvalue_5)) _asbga_keys_and_locks_lock)
                      (Variables|/KeysAndLocks/Items/KeysAndLocks/BGA_KeysAndLocks_Lock.BGA_KeysAndLocks_Lock_C|SetInteractionText_NeedKey (Utilities|Text|FormatText "INVTEXT(\"{0}\\r\\n{1} {2}\")" (select _returnvalue_6 (Utilities|Text|FormatText "NSLOCTEXT(\"[62AC31B24745139BE344D2B15B75F839]\", \"5C85BE1B4C0B44D72BBF689B74C0B055\", \"{ModSetName}\\r\\n{WeaponName}\")" _modsetname (ItemDefinition|GetDisplayName _itemdefinition)) (ItemDefinition|GetDisplayName _itemdefinition)) _result _returnvalue_5) _asbga_keys_and_locks_lock)
                      (return "" ""))) ; With replicated variables it is necessary to check validity before casting.
                ))))))
      (:else))))

(fn CenterSkeletalMesh
  (Transformation|SetWorldLocation (Variables|Default|GetItemMesh(Skeletal)) (+ (Transformation|GetWorldLocation (Variables|Default|GetItemMesh(Skeletal))) (* (- (Transformation|GetWorldLocation (Variables|Default|GetItemMesh(Skeletal))) (Collision|GetComponentBounds (Variables|Default|GetItemMesh(Skeletal)))) 0.5)) false true))

(fn GetRandomRareOrLegendaryWeapon
  (bind (_outloottodrop _returnvalue) (Loot|PickLootDrops (Variables|Loot|GetTier%20Group%20Name) 0 -1 "(GameplayTags=)"))
  (if _returnvalue
    (bind _output_get (LocalVariables|SetCurrLoot _outloottodrop))
    (bind _array_element (Utilities|Array|ForEachLoopwithBreak _output_get)
      (:LoopBody
        (bind _asfort_weapon_item_definition (Utilities|Casting|CastToFortWeaponItemDefinition (Class|FortItemFunctionLibrary|BreakFortItemEntry _array_element)))
        (bind _itemdata (Class|ItemDefinitionBase|GetItemData _asfort_weapon_item_definition)
          (:Found
            (if (and (Utilities|IsValidSoftObjectReference (Utilities|Struct|BreakFortnitePickupItemData _itemdata)) (not (GameplayTags|HasMatchingGameplayTag _asfort_weapon_item_definition "(TagName=\"Item.Ammo\")")))
              ;[0.07717 1 0.04473] Branching if moddable weapon logic as creative etc. using older non-modded weapons
              (if (GameplayTags|HasMatchingGameplayTag _asfort_weapon_item_definition "(TagName=\"ModSetBucket\")")
                (bind _returnvalue_2 (Class|FortWeaponModFunctionLibrary|GetRandomWeaponModSetData _asfort_weapon_item_definition))
                (Variables|ModSetData|SetWeponModSetData _returnvalue_2)
                (if (Utilities|Array|ContainsItem (Variables|/KeysAndLocks/Items/KeysAndLocks/BGA_KeysAndLocks_DisplayCasesManager.BGA_KeysAndLocks_DisplayCasesManager_C|GetGeneratedItems (Variables|Default|GetSpawner)) (Weapon|GetUpgradedWeaponItemVerticaltoRarity _asfort_weapon_item_definition "Legendary"))
                  (LocalVariables|SetLastValidWeapon _asfort_weapon_item_definition))
                ;[1 0.5052 0.5052] Found an item that fits our spec, return it.
                (return _asfort_weapon_item_definition (LocalVariables|GetCurrLoot)))
              (Components|DestroyComponent (Variables|Default|GetCaseModMesh1))))))
      (:Completed
        _))
    (bind _result (Math|Integer|IncrementInt (LocalVariables|GetAttemtps))
      (:%20%20
        (if (> _result 10)
          _
          (else
            _))
        ;[1 0.5052 0.5052] Failed to find a valid item after a failsafe number of attempts, return the last valid item.
        (return (LocalVariables|GetLastValidWeapon) (LocalVariables|GetCurrLoot)))))
  ;[1 0.5052 0.5052] Failed to find item, return nothing.
  (return 0 (LocalVariables|GetCurrLoot)))

(fn GetLockChildActors
  (return (Utilities|Array|MakeArray (Variables|Default|GetLockChildActor))))

(fn InitFXRarityColor
  (bind _itemdefinition (Variables|Default|GetItemDefinition))
  (Utilities|IsValid _itemdefinition
    (:Is%20Valid
      (Variables|Colors|SetFXRarityColor (Utilities|Array|Get(aref) (Variables|Colors|GetRarityColor) (Math|Conversions|ToInteger(Byte) (Item|GetRarity _itemdefinition)))))))

(fn IsOpen
  (return (Variables|Loot|Getb_LootSpawned)))

(fn Log
  (params (FunctionName :type String) (Message :type String))
  (if (Variables|Default|GetbLoggingEnabled%20Const)
    (Fortnite|FortShippingLog self (Utilities|String|ToString(Text) (Utilities|Text|FormatText "INVTEXT(\"BGA_KeysAndLocks_DisplayCase::{FunctionName}() object: [{ObjectName}] {Message}.\")" FunctionName (Utilities|GetObjectName self) Message)))))

(fn OnRep_ItemDefinition
  (CallFunction|UpdateItemDefinition))

(fn SendOpenQuestEvent
  (bind _lockchildactor (Variables|Default|GetLockChildActor))
  (bind _childactor (Variables|/Script/Engine.ChildActorComponent|GetChildActor _lockchildactor))
  (Utilities|IsValid _childactor
    (:Is%20Valid
      (bind _asbga_keys_and_locks_lock (Utilities|Casting|CastToBGA_KeysAndLocks_Lock _childactor))
      (Utilities|IsValid (Fortnite|GetFortPlayerControllerfromActor (Variables|/Game/Athena/Items/EnvironmentalItems/Locks/BGA_Athena_Lock_Parent.BGA_Athena_Lock_Parent_C|GetLastInteractingPawn _asbga_keys_and_locks_lock))
        (:Is%20Valid
          (Utilities|IsValid (Quests|GetQuestManager (Fortnite|GetFortPlayerControllerfromActor (Variables|/Game/Athena/Items/EnvironmentalItems/Locks/BGA_Athena_Lock_Parent.BGA_Athena_Lock_Parent_C|GetLastInteractingPawn _asbga_keys_and_locks_lock)) "Athena")
            (:Is%20Valid
              (Quest|SendCustomStatEventwithTags (Quests|GetQuestManager (Fortnite|GetFortPlayerControllerfromActor (Variables|/Game/Athena/Items/EnvironmentalItems/Locks/BGA_Athena_Lock_Parent.BGA_Athena_Lock_Parent_C|GetLastInteractingPawn _asbga_keys_and_locks_lock)) "Athena") "Interact" (Variables|Quest|GetAdditional%20Quest%20Tags%20To%20Send%20On%20Open) "(GameplayTags=((TagName=\"Building.Type.Container.TreasureChest.Locked\")))" 1))))))) ; With replicated variables it is necessary to check validity before casting.
)

(fn SetMeshAndRarityColor
  (params (Mesh%20Component :type Static%20Mesh%20Component%20Object%20Reference) (Mesh :type Static%20Mesh%20Object%20Reference))
  (Components|StaticMesh|SetStaticMesh Mesh%20Component Mesh)
  (CallFunction|SetProjectionRarityColor (Item|GetRarity (Variables|Default|GetItemDefinition)) Mesh%20Component)
  (Transformation|SetRelativeTransform Mesh%20Component (Transformation|GetBoneTransform (Variables|Default|GetItemMesh(Skeletal)) (Utilities|String|StringToName (Variables|Default|GetModBoneName)) "RTS_Actor") false false)
  (Transformation|AddLocalOffset Mesh%20Component (Variables|Default|GetModBoneOffset) false false))

(fn SetProjectionRarityColor
  (params (Rarity :type EFortRarity%20Enum :default "Common") (Component :type Primitive%20Component%20Object%20Reference))
  (bind _returnvalue (Math|Conversions|ToInteger(Byte) Rarity))
  (bind _raritycolor (Variables|Colors|GetRarityColor))
  (bind _output (Utilities|Array|Get(aref) _raritycolor _returnvalue))
  (Utilities|IsValid Component
    (:Is%20Valid
      (bind _returnvalue_2 (Rendering|Material|CreateDynamicMaterialInstance "/KeysAndLocks/Items/KeysAndLocks/FX/MI_Hologram_EdgeLit_v2.MI_Hologram_EdgeLit_v2" "None" "None" :!class /Script/Engine.KismetMaterialLibrary))
      (for _index (range (+ (Rendering|Material|GetNumMaterials Component) 1))
        (Rendering|Material|SetMaterial Component _index _returnvalue_2))
      (Rendering|Material|SetVectorParameterValue _returnvalue_2 "Color A" (Utilities|Array|Get(aref) (Variables|Colors|GetRarityColorA) _returnvalue) :!class /Script/Engine.MaterialInstanceDynamic)
      (Rendering|Material|SetVectorParameterValue _returnvalue_2 "Color B" _output :!class /Script/Engine.MaterialInstanceDynamic)
      (bind _output_get (Variables|Colors|SetBeamColor _output))
      (bind _ns_keyslocks_holo_beam (Variables|Default|GetNS_KeysLocks_Holo_Beam)
        (:then
          (Effects|Components|ParticleSystem|SetColorParameter _ns_keyslocks_holo_beam "Beam Color" (Math|Color|MakeColor (Math|Color|BreakColor _output_get) (.G (Math|Color|BreakColor _output_get)) (.B (Math|Color|BreakColor _output_get)) (Variables|Hologram|GetBeamOpacityWhenPlayerTooFar))))
        (:else)))))

(fn UpdateItemDefinition
  (Networking|CanExecuteCosmeticEvents
    (:True
      (CallFunction|UpdateLockChildActors)
      (CallFunction|LoadMesh)
      (CallFunction|InitFXRarityColor))))

(fn UpdateLockChildActors
  (bind _lock_child_actors (CallFunction|GetLockChildActors))
  (Networking|CanExecuteCosmeticEvents
    (:True
      (Utilities|FlowControl|Sequence
        (:then_0
          (bind _returnvalue (CallFunction|NativeLockChildActorsPassAllRequirements _lock_child_actors (Variables|Default|GetItemDefinition)))
          (LocalVariables|SetbAllRequirementsPassed _returnvalue))
        (:then_1
          (if (LocalVariables|GetbAllRequirementsPassed)
            ;[0 0.1495 1] Our requirements have been met, update the lock child actors.
            (Utilities|Time|ClearandInvalidateTimerbyHandle (Variables|Default|GetUpdateLockChildActorsTimerHandle))
            (if (ScalableFloat|GetValueasBool (Variables|Text|GetUseBPStringFormat))
              (CallFunction|CalculateLockInteractStrings _lock_child_actors)
              (else
                (bind (_havekey _needkey) (CallFunction|NativeCalculateLockInteractStrings (Variables|Loot|GetLoot) (Variables|Text|GetRarityTexts) (Variables|Default|GetItemDefinition) (Variables|Text|GetShowCount) (Variables|ModSetData|GetWeponModSetData)))
                (for _array_element _lock_child_actors
                  (Utilities|IsValid _array_element
                    (:Is%20Valid
                      (Utilities|IsValid (Variables|/Script/Engine.ChildActorComponent|GetChildActor _array_element)
                        (:Is%20Valid
                          (bind _asbga_keys_and_locks_lock (Utilities|Casting|CastToBGA_KeysAndLocks_Lock (Variables|/Script/Engine.ChildActorComponent|GetChildActor _array_element)))
                          (Variables|/KeysAndLocks/Items/KeysAndLocks/BGA_KeysAndLocks_Lock.BGA_KeysAndLocks_Lock_C|SetInteractionText_HaveKey _havekey _asbga_keys_and_locks_lock)
                          (Variables|/KeysAndLocks/Items/KeysAndLocks/BGA_KeysAndLocks_Lock.BGA_KeysAndLocks_Lock_C|SetInteractionText_NeedKey _needkey _asbga_keys_and_locks_lock))) ; With replicated variables it is necessary to check validity before casting.
                    )))))
            (elif (Utilities|Time|IsValidTimerHandle (Variables|Default|GetUpdateLockChildActorsTimerHandle))
              _
              (else
                ;[0 0.9526 1] Our requirements have NOT been met, run this function again in the near future.
                ; This is not an ideal solution, but our options are limited since ChildActorComponent takes a while to replicate the necessary data and has no callbacks for when the replication takes place. We have no other choice but to update on an interval.
                (Variables|Default|GetItemDefinition
                  (:then
                    (bind _returnvalue_2 (Utilities|Time|SetTimerbyFunctionName 0 "UpdateLockChildActors" (Math|Float|Max(Float) 0.001 (Variables|Default|GetUpdateLockChildActorsInterval%20Const)) true true))
                    (Variables|Default|SetUpdateLockChildActorsTimerHandle _returnvalue_2))
                  (:else)) ; Only set the update timer if the item definition is valid. If the ItemDefinition is invalid and becomes valid OnRep_ItemDefinition will manually call the update function.
              ))))))))

(fn ConstructionScript
  (|CallParentFunction)
  (Variables|Hologram|SetHologram%20Scene%20Components (Utilities|Array|MakeArray (Variables|Default|GetItemMesh(Skeletal)) (Variables|Default|GetItemMesh(Static)) (Variables|Default|GetNS_KeysLocks_Holo_Beam) (Variables|Default|GetCaseModMesh1) (Variables|Default|GetCaseModMesh2) (Variables|Default|GetCaseModMesh3) (Variables|Default|GetCaseModMesh4))))
