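; Wild display case variant for Keys and Locks, labelled as a Mystery Box and projected in Unattainable rarity color.
; Overrides the parent's random weapon pick to upgrade it to Legendary.
; Server-only loot spawning rolls a 25% chance at the exotics loot table, otherwise dropping one to three pickups of the case's item in front of and above the lock.
; The inline roll is marked temporary, pending a custom loot tier.
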(event EventBeginPlay
  (|CallParentFunction)
  (CallFunction|SetProjectionRarityColor "Unattainable" (Variables|Default|GetItemMesh(Static))))

(event Collision|EventActorBeginOverlap
  (params (OtherActor :type Actor%20Object%20Reference)))

(event EventSpawnLoot(ServerOnly)
  ;[1 1 1] Temporary -- should be replaced with just using custom loot tier
  (if (> (Math|Random|RandomFloat) 0.75)
    (Class|AthenaFunctionLibrary|SpawnLootfromTable "Loot_DisplayCase_Exotics" 1 -1 (Utilities|Operators|Add (Transformation|GetWorldLocation (Variables|Default|GetLockChildActor)) "0,0,65.000000" (* (Math|Rotator|GetRotationXVector (Transformation|GetWorldRotation (Variables|Default|GetLockChildActor))) 3e+01)) (Math|Rotator|GetRotationXVector (Transformation|GetWorldRotation (Variables|Default|GetLockChildActor))) 0 true)
    (else
      (for _ (range (+ (Math|Random|RandomIntegerinRange 0 2) 1))
        (Inventory|SpawnPickupInWorld (Variables|Default|GetItemDefinition) 1 (Utilities|Operators|Add (Transformation|GetWorldLocation (Variables|Default|GetLockChildActor)) "0,0,65.000000" (* (Math|Rotator|GetRotationXVector (Transformation|GetWorldRotation (Variables|Default|GetLockChildActor))) 3e+01)) (Math|Rotator|GetRotationXVector (Transformation|GetWorldRotation (Variables|Default|GetLockChildActor))) 0 true false false 0 "Other" "Unset" 0 false)))))

(fn CalculateLockInteractStrings
  (params (Lock%20Child%20Actors :type Array%20of%20Child%20Actor%20Component%20Object%20References))
  (return (Utilities|Text|FormatText "INVTEXT(\"Use Key\\r\\n{1}\")" (Utilities|Text|MakeLiteralText "NSLOCTEXT(\"[24DF5C084CF7A5CCC38944AC87230AE3]\", \"6439C58746FDBDCD3FFB5480D9193EF7\", \"Mystery Box\")")) (Utilities|Text|MakeLiteralText "NSLOCTEXT(\"[24DF5C084CF7A5CCC38944AC87230AE3]\", \"6439C58746FDBDCD3FFB5480D9193EF7\", \"Mystery Box\")")))

(fn GetRandomRareOrLegendaryWeapon
  (bind (_asfort_weapon_item_definition _loot) (|CallParentFunction))
  (return (Weapon|GetUpgradedWeaponItemVerticaltoRarity _asfort_weapon_item_definition "Legendary") _loot))

(fn ConstructionScript
  (|CallParentFunction))
