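; Keys and Locks manager actor that spawns a weighted-random set of display cases and a minimap icon on the server.
; It tracks each case's lock, replicating an all-opened flag whose OnRep clears markers and disables hologram triggers.
; Cosmetic overlaps and local view-target changes drive timer-staggered weapon hologram reveals and hides, covering spectated pawns.
; A client ping trigger manages per-player markers for key holders, removed on exit or key-tag loss.
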(event EventBeginPlay
  (bind _spawningruleswithweight (Variables|Default|GetSpawningRulesWithWeight))
  ;[0 1 0.04169] BeginPlay
  (|CallParentFunction)
  (Networking|SwitchHasAuthority
    (:Authority
      (Utilities|FlowControl|Sequence
        (:then_0
          ;;[1 0.3562 0] Disable collision on the server.
          (Collision|SetCollisionEnabled (Variables|Default|GetPlayerTriggerForPing) "NoCollision")
          (bind _returnvalue (Game|SpawnActorfromClass "/KeysAndLocks/Items/KeysAndLocks/KeysAndLocks_DisplayCasesMapIcon.KeysAndLocks_DisplayCasesMapIcon_C" (Transformation|GetActorTransform) "Undefined" "OverrideRootScale" 0 0))
          (Variables|Default|SetMiniMapMarkerActor _returnvalue))
        (:then_1
          ;;[0 0.06364 1] Spawn 1 or more display cases
          (bind _values (Utilities|Map|Values _spawningruleswithweight))
          (bind _output_get (Variables|Default|SetChosenIndex (Math|Fortnite|GetRandomWeightedIndex (CallFunction|ScalableFloatArraytoFloatArray _values))))
          (bind _keys (Utilities|Map|Keys _spawningruleswithweight))
          (if (Utilities|Array|IsValidIndex _keys _output_get)
            (for _array_element (Class|GetClassDefaults (Utilities|Array|Get(acopy) _keys _output_get))
              (bind _spawnedthing (CallFunction|SpawnActorViaSpawningData _array_element))
              (bind _filteredarray (Utilities|Array|FilterArray (Actor|GetAllChildActors _spawnedthing false) "/KeysAndLocks/Items/KeysAndLocks/BGA_KeysAndLocks_Lock.BGA_KeysAndLocks_Lock_C"))
              (bind _asbga_keys_and_locks_lock (Utilities|Casting|CastToBGA_KeysAndLocks_Lock (Utilities|Array|Get(acopy) _filteredarray)))
              (Default|BindEventtoOnMultilockCompleted _asbga_keys_and_locks_lock (delegate Custom|LockOpened))
              (Utilities|Array|AddUnique (Variables|Default|GetLocksToBeOpened) _asbga_keys_and_locks_lock))))))
    (:Remote
      (Utilities|FlowControl|Sequence
        (:then_0
          ;;[1 0.8122 0.2487] On clients if all of the managed display cases have been opened, then shut down all collision.
          (if (Variables|Default|GetAllManagedDisplayCasesOpened)
            (Collision|SetCollisionEnabled (Variables|Default|GetPlayerTriggerForPing) "NoCollision")))
        (:then_1
          ;[0.1936 1 0.9203] Monitor view target changes of all local players.
          (for _array_element_2 (Fortnite|GetLocalFortPlayerControllers true true)
            (Utilities|IsValid _array_element_2
              (:Is%20Valid
                (EventDispatchers|BindEventtoOnViewTargetChanged _array_element_2 (EventDispatchers|CreateEvent :!event Custom|OnLocalControllerViewTargetChanged))))))))))

(event EventEndPlay
  (params (EndPlayReason :type EEndPlayReason%20Enum :default "Destroyed"))
  ;[1 0.4 0.01108] EndPlay
  (|CallParentFunction EndPlayReason)
  (Utilities|FlowControl|Sequence
    (:then_0
      (Utilities|Time|ClearandInvalidateTimerbyHandle (Variables|Default|GetShowWeaponsTimer))
      (CallFunction|DestroyAllPlayerMarkers)
      (bind _minimapmarkeractor (Variables|Default|GetMiniMapMarkerActor)
        (:then
          (Actor|DestroyActor _minimapmarkeractor)
          (Variables|Default|SetMiniMapMarkerActor 0))
        (:else)))
    (:then_1
      (for _array_element (Fortnite|GetLocalFortPlayerControllers true true)
        (Utilities|IsValid _array_element
          (:Is%20Valid
            (EventDispatchers|UnbindEventfromOnViewTargetChanged _array_element (EventDispatchers|CreateEvent :!event Custom|OnLocalControllerViewTargetChanged))))))))

(event Custom|LockOpened
  (params (Self%20Lock :type BGA%20Keys%20And%20Locks%20Lock%20Object%20Reference))
  ;[0 0.009654 1] LockOpened
  (Utilities|Array|RemoveItem (Variables|Default|GetLocksToBeOpened) Self%20Lock)
  (if (<= (Utilities|Array|Length (Variables|Default|GetLocksToBeOpened)) 0)
    (Utilities|FlowControl|DoOnce false
      (:Completed
        (Variables|Default|SetAllManagedDisplayCasesOpened true)
        (Networking|FlushNetDormancy)))))

(event OnComponentBeginOverlap(PlayerTriggerForHolograms)
  (params (OverlappedComponent :type Primitive%20Component%20Object%20Reference) (OtherActor :type Actor%20Object%20Reference) (OtherComp :type Primitive%20Component%20Object%20Reference) (OtherBodyIndex :type Integer) (bFromSweep :type Boolean) (SweepResult :type Hit%20Result%20Structure%20%28by%20ref%29))
  ;[0 1 0.6991] Overlap Begin/End PlayerTriggerForHolograms
  (Networking|CanExecuteCosmeticEvents
    (:True
      (bind _aspawn (Utilities|Casting|CastToPawn OtherActor))
      (Utilities|Array|Add (Variables|Default|GetPawnsWithinHologramRange) _aspawn)
      (bind _returnvalue (CallFunction|IsPawnALocalPlayerViewTarget _aspawn))
      (if _returnvalue
        (CallFunction|ShowHolograms)))))

(event OnComponentEndOverlap(PlayerTriggerForHolograms)
  (params (OverlappedComponent :type Primitive%20Component%20Object%20Reference) (OtherActor :type Actor%20Object%20Reference) (OtherComp :type Primitive%20Component%20Object%20Reference) (OtherBodyIndex :type Integer))
  ;[0 1 0.6991] Overlap Begin/End PlayerTriggerForHolograms
  (Networking|CanExecuteCosmeticEvents
    (:True
      (bind _aspawn (Utilities|Casting|CastToPawn OtherActor))
      (Utilities|Array|RemoveItem (Variables|Default|GetPawnsWithinHologramRange) _aspawn)
      (bind _oneormorepawnsarewithinrange (CallFunction|AreAnyViewTargetsWithinHologramRange))
      (if (not _oneormorepawnsarewithinrange)
        (CallFunction|HideHolograms)))))

(event Custom|ShowHolograms
  ;[1 0 0.6794] ShowHolograms() turns on hologram cosmetics and starts a timer to cycle between available weapons
  ; ShowNextWeapon() called on a timer, cycles between available weapons.
  ; HideHolograms() turns off hologram cosmetics
  ; HideNextWeapon() called on a timer, cycles backwards through the available weapons until it hits the first weapon and then stops.
  (if (not (Variables|Default|GetIsShowingHolograms))
    (bind _returnvalue (Utilities|Time|SetTimerbyEvent (delegate Custom|Show%20Next%20Weapon) 0.1 true))
    (Variables|Default|SetShowWeaponsTimer _returnvalue)
    (Utilities|Time|ClearandInvalidateTimerbyHandle (Variables|Default|GetHideWeaponsTimer))
    (CallFunction|ShowNextWeapon)
    (Variables|Default|SetIsShowingHolograms true)))

(event Custom|Show%20Next%20Weapon
  (bind _returnvalue (Utilities|Array|Length (Variables|Default|GetSpawnedDisplayCases)))
  (bind _weaponshowingindex (Variables|Default|GetWeaponShowingIndex))
  (bind _spawneddisplaycases (Variables|Default|GetSpawnedDisplayCases))
  (bind _output (Utilities|Array|Get(acopy) _spawneddisplaycases _weaponshowingindex))
  (bind _weaponshowingindex_2 (Variables|Default|GetWeaponShowingIndex))
  ;[1 0 0.6794] ShowHolograms() turns on hologram cosmetics and starts a timer to cycle between available weapons
  ; ShowNextWeapon() called on a timer, cycles between available weapons.
  ; HideHolograms() turns off hologram cosmetics
  ; HideNextWeapon() called on a timer, cycles backwards through the available weapons until it hits the first weapon and then stops.
  (Utilities|IsValid _output
    (:Is%20Valid
      (Class|BGAKeysandLocksDisplayCase|ShowWeapon _output))
    (:Is%20Not%20Valid
      _))
  (Math|Integer|IncrementInt _weaponshowingindex_2
    (:%20%20
      (if (>= _weaponshowingindex_2 _returnvalue)
        (Utilities|Time|ClearandInvalidateTimerbyHandle (Variables|Default|GetShowWeaponsTimer))
        (Variables|Default|SetWeaponShowingIndex (- _returnvalue 1))))))

(event Custom|HideHolograms
  ;[1 0 0.6794] ShowHolograms() turns on hologram cosmetics and starts a timer to cycle between available weapons
  ; ShowNextWeapon() called on a timer, cycles between available weapons.
  ; HideHolograms() turns off hologram cosmetics
  ; HideNextWeapon() called on a timer, cycles backwards through the available weapons until it hits the first weapon and then stops.
  (if (Variables|Default|GetIsShowingHolograms)
    (bind _returnvalue (Utilities|Time|SetTimerbyEvent (delegate Custom|Hide%20Next%20Weapon) 0.1 true))
    (Variables|Default|SetHideWeaponsTimer _returnvalue)
    (Utilities|Time|ClearandInvalidateTimerbyHandle (Variables|Default|GetShowWeaponsTimer))
    (Variables|Default|SetIsShowingHolograms false)))

(event Custom|Hide%20Next%20Weapon
  (bind _weaponshowingindex (Variables|Default|GetWeaponShowingIndex))
  (bind _spawneddisplaycases (Variables|Default|GetSpawnedDisplayCases))
  (bind _output (Utilities|Array|Get(acopy) _spawneddisplaycases _weaponshowingindex))
  (bind _weaponshowingindex_2 (Variables|Default|GetWeaponShowingIndex))
  ;[1 0 0.6794] ShowHolograms() turns on hologram cosmetics and starts a timer to cycle between available weapons
  ; ShowNextWeapon() called on a timer, cycles between available weapons.
  ; HideHolograms() turns off hologram cosmetics
  ; HideNextWeapon() called on a timer, cycles backwards through the available weapons until it hits the first weapon and then stops.
  (Utilities|IsValid _output
    (:Is%20Valid
      (Class|BGAKeysandLocksDisplayCase|HideWeapon _output))
    (:Is%20Not%20Valid
      _))
  (Math|Integer|DecrementInt _weaponshowingindex_2
    (:%20%20
      (if (< _weaponshowingindex_2 0)
        (Utilities|Time|ClearandInvalidateTimerbyHandle (Variables|Default|GetHideWeaponsTimer))
        (Variables|Default|SetWeaponShowingIndex 0)))))

(event OnComponentBeginOverlap(PlayerTriggerForPing)
  (params (OverlappedComponent :type Primitive%20Component%20Object%20Reference) (OtherActor :type Actor%20Object%20Reference) (OtherComp :type Primitive%20Component%20Object%20Reference) (OtherBodyIndex :type Integer) (bFromSweep :type Boolean) (SweepResult :type Hit%20Result%20Structure%20%28by%20ref%29))
  (bind _playercontrollerstomarkeridsmap (Variables|Default|GetPlayerControllersToMarkerIdsMap))
  ;[0 0.2091 1] Overlap Begin/End PlayerTriggerForPing
  (Networking|SwitchHasAuthority
    (:Remote
      (bind _returnvalue (GameplayTags|HasMatchingGameplayTag(Message) OtherActor "(TagName=\"Granted.ItemEquipped.Key\")"))
      (if (and _returnvalue (not (Variables|Default|GetAllManagedDisplayCasesOpened)))
        (bind _aspawn (Utilities|Casting|CastToPawn OtherActor))
        (if (Pawn|IsLocallyControlled _aspawn)
          (bind _asfort_player_controller_athena (Utilities|Casting|CastToFortPlayerControllerAthena (Pawn|GetController _aspawn)))
          (CallFunction|DestroyPlayersDisplayCaseMarker _asfort_player_controller_athena)
          (if (Utilities|IsValid (Class|FortPlayerController|GetMarkerComponent _asfort_player_controller_athena) :!class /Script/Engine.KismetSystemLibrary)
            (Utilities|IsValid (MarkerComponent|FindMarkerbyMarkedActor (Class|FortPlayerController|GetMarkerComponent _asfort_player_controller_athena) self)
              (:Is%20Valid
                (bind (_playerbankid _bankinstanceid) (Fortnite|RegisterSoundIndicatorPlacementContext _asfort_player_controller_athena self "0, 0, 0" false))
                (bind _asynctaskproxy (Ability|Tasks|WaitGameplayTagRemoveFromActor _aspawn "(TagName=\"Granted.ItemEquipped.Key\")")
                  (:then
                    (Utilities|Map|Add _playercontrollerstomarkeridsmap _asfort_player_controller_athena (Utilities|Struct|MakeSTRUCTKeysandLocksDisplayCaseMarkerData (Utilities|Struct|MakeMarkerID _playerbankid _bankinstanceid) _asynctaskproxy)))
                  (:Removed
                    (bind _keys (Utilities|Map|Keys _playercontrollerstomarkeridsmap))
                    (bind _values (Utilities|Map|Values _playercontrollerstomarkeridsmap))
                    (bind (_array_element _array_index) (Utilities|Array|ForEachLoopwithBreak _keys)
                      (:LoopBody
                        (if (== _asynctaskproxy (.AsyncWait_5_96B390A04F5599957513A4B47883217C (Utilities|Struct|BreakSTRUCTKeysandLocksDisplayCaseMarkerData (Utilities|Array|Get(aref) _values _array_index))))
                          (CallFunction|DestroyPlayersDisplayCaseMarker _array_element))))))))))))))

(event OnComponentEndOverlap(PlayerTriggerForPing)
  (params (OverlappedComponent :type Primitive%20Component%20Object%20Reference) (OtherActor :type Actor%20Object%20Reference) (OtherComp :type Primitive%20Component%20Object%20Reference) (OtherBodyIndex :type Integer))
  ;[0 0.2091 1] Overlap Begin/End PlayerTriggerForPing
  (Networking|SwitchHasAuthority
    (:Remote
      (bind _aspawn (Utilities|Casting|CastToPawn OtherActor))
      (bind _asfort_player_controller_athena (Utilities|Casting|CastToFortPlayerControllerAthena (Pawn|GetController _aspawn)))
      (CallFunction|DestroyPlayersDisplayCaseMarker _asfort_player_controller_athena))))

(event Custom|Destroy%20Players%20Display%20Case%20Marker
  (params (PlayerController :type Fort%20Player%20Controller%20Athena%20Object%20Reference))
  (bind _playercontrollerstomarkeridsmap (Variables|Default|GetPlayerControllersToMarkerIdsMap))
  (bind (_value _returnvalue) (Utilities|Map|Find _playercontrollerstomarkeridsmap PlayerController))
  (bind (_markerid_4_bb966b5e45d4f4a091a5688239f53231 _asyncwait_5_96b390a04f5599957513a4b47883217c) (Utilities|Struct|BreakSTRUCTKeysandLocksDisplayCaseMarkerData _value))
  (bind _returnvalue_2 (Class|FortPlayerController|GetMarkerComponent PlayerController))
  ;[1 0 0.9068 0.458] Destroy Current Ping
  (Networking|CanExecuteCosmeticEvents
    (:True
      (Utilities|IsValid PlayerController
        (:Is%20Valid
          (if _returnvalue
            (AsyncAction|Cancel _asyncwait_5_96b390a04f5599957513a4b47883217c)
            (Utilities|IsValid _returnvalue_2
              (:Is%20Valid
                (MarkerComponent|UnmarkActoronClient _returnvalue_2 _markerid_4_bb966b5e45d4f4a091a5688239f53231))
              (:Is%20Not%20Valid
                _))
            (Utilities|Map|Remove _playercontrollerstomarkeridsmap PlayerController)))))))

(event Custom|OnLocalControllerViewTargetChanged
  (params (PlayerController :type Fort%20Player%20Controller%20Object%20Reference) (OldViewTarget :type Actor%20Object%20Reference) (NewViewTarget :type Actor%20Object%20Reference))
  ;[0.5943 0.04928 1] OnLocalControllerViewTargetChanged() Updates if the holograms should be shown/hidden based on if any of the pawns within range are view targets of any local player.
  ;;[0.07181 1 0.3179] If the player's new view target is one of the pawns within hologram range, show the holograms.
  (bind _aspawn (Utilities|Casting|CastToPawn NewViewTarget)
    (:then
      (if (Utilities|Array|ContainsItem (Variables|Default|GetPawnsWithinHologramRange) _aspawn)
        (CallFunction|ShowHolograms)
        (else
          _)))
    (:CastFailed
      _))
  ;;[1 0.7802 0.2334] If the player's previous view target was within hologram range AND no local view targets are within hologram range, then hide the holograms.
  (if (Variables|Default|GetIsShowingHolograms)
    (bind _aspawn_2 (Utilities|Casting|CastToPawn OldViewTarget))
    (if (Utilities|Array|ContainsItem (Variables|Default|GetPawnsWithinHologramRange) _aspawn_2)
      (bind _oneormorepawnsarewithinrange (CallFunction|AreAnyViewTargetsWithinHologramRange))
      (if (not _oneormorepawnsarewithinrange)
        (CallFunction|HideHolograms)))))

(fn AreAnyViewTargetsWithinHologramRange
  (bind _pawnswithinhologramrange (Variables|Default|GetPawnsWithinHologramRange))
  (Utilities|Array|RemoveItem _pawnswithinhologramrange) ; Remove all null ptr entries.
  (for _array_element _pawnswithinhologramrange
    (bind _returnvalue (CallFunction|IsPawnALocalPlayerViewTarget _array_element))
    (if _returnvalue
      ;[1 0.5052 0.5052] We found a pawn within range that is a local player's view target.
      (return true)))
  ;[1 0.5052 0.5052] No pawns found that match our criteria.
  (return false))

(fn DestroyAllPlayerMarkers
  (bind _playercontrollerstomarkeridsmap (Variables|Default|GetPlayerControllersToMarkerIdsMap))
  (Networking|CanExecuteCosmeticEvents
    (:True
      (bind _keys (Utilities|Map|Keys _playercontrollerstomarkeridsmap))
      (bind _output_get (LocalVariables|SetPlayersToClearMarkersFrom _keys))
      (for _array_element _output_get
        (CallFunction|DestroyPlayersDisplayCaseMarker _array_element))
      (Utilities|Map|Clear _playercontrollerstomarkeridsmap))))

(fn IsPawnALocalPlayerViewTarget
  (params (Pawn :type Pawn%20Object%20Reference))
  (Utilities|FlowControl|Sequence
    (:then_0
      (if (not (Pawn|IsLocallyControlled Pawn))
        ;[1 0.9672 0.01729] Check if any local controller is using this pawn as a view target.
        (for _array_element (Fortnite|GetLocalFortPlayerControllers true true)
          (Utilities|IsValid _array_element
            (:Is%20Valid
              (if (== (Pawn|GetViewTarget _array_element) Pawn)
                ;[1 0.5052 0.5052] This pawn is a local controller's view target, return true.
                (return true)))))))
    (:then_1
      ;[1 0.5052 0.5052] Failed to find any local players that are using this pawn as a view target. Return false.
      (return false))))

(fn OnRep_AllManagedDisplayCasesOpened
  (if (Variables|Default|GetAllManagedDisplayCasesOpened)
    (CallFunction|DestroyAllPlayerMarkers)
    (Collision|SetCollisionEnabled (Variables|Default|GetPlayerTriggerForHolograms) "NoCollision")
    ;[1 0.5326 0] Remove minimap marker on the server and on clients (destroying it on the server lightens the server's load, destroy on the client removes the component's map marker.
    (bind _minimapmarkeractor (Variables|Default|GetMiniMapMarkerActor)
      (:then
        (Actor|DestroyActor _minimapmarkeractor)
        (Variables|Default|SetMiniMapMarkerActor 0))
      (:else))))

(fn ScalableFloatArrayToFloatArray
  (params (Array :type Array%20of%20Scalable%20Float%20Structures))
  (for _array_element Array
    (Utilities|Array|Add (LocalVariables|GetResult) (ScalableFloat|GetValueatLevel _array_element)))
  (return (LocalVariables|GetResult)))

(fn SpawnActorViaSpawningData
  (params (STRUCT_DisplayCase_SpawningData :type STRUCT%20Keys%20And%20Locks%20Display%20Case%20Spawning%20Data%20Structure%20%28by%20ref%29))
  (bind (_bgaclasstospawnwithweightmap_13_9358d96849ac9a5e5ea375b2e2dd29e7 _relativetransform_8_a5fa37a74934cb0179d6dbbab692ea28) (Utilities|Struct|BreakSTRUCTKeysandLocksDisplayCaseSpawningData STRUCT_DisplayCase_SpawningData))
  (bind (_askeys_and_locks_chest _bsuccess) (Utilities|Casting|CastToKeysAndLocks_ChestClass (LocalVariables|GetClass)))
  (bind _class (LocalVariables|GetClass))
  (bind _values (Utilities|Map|Values _bgaclasstospawnwithweightmap_13_9358d96849ac9a5e5ea375b2e2dd29e7))
  (bind _keys (Utilities|Map|Keys _bgaclasstospawnwithweightmap_13_9358d96849ac9a5e5ea375b2e2dd29e7))
  (if (Utilities|Array|IsValidIndex _keys (Math|Fortnite|GetRandomWeightedIndex _values))
    (LocalVariables|SetClass (Utilities|Array|Get(acopy) _keys (Math|Fortnite|GetRandomWeightedIndex _values)))
    (if (Utilities|IsValidClass (LocalVariables|GetClass))
      (Utilities|Casting|CastToBGA_KeysAndLocks_DisplayCase_2LocksClass _class
        (:then
          (Variables|Default|SetHasExoticChest true))
        (:CastFailed
          _))
      (bind _returnvalue (Game|SpawnActorfromClass _class (Math|Transform|SelectTransform (Math|Transform|ComposeTransforms (struct :Location "0, 0, 0" :Rotation "0,-90.000000, 0" :Scale "1.000000,1.000000,1.000000") (Math|Transform|ComposeTransforms _relativetransform_8_a5fa37a74934cb0179d6dbbab692ea28 (Transformation|GetActorTransform))) (Math|Transform|ComposeTransforms _relativetransform_8_a5fa37a74934cb0179d6dbbab692ea28 (Transformation|GetActorTransform)) _bsuccess) "Undefined" "SelectDefaultAtRuntime" 0 0))
      (Utilities|FlowControl|Sequence
        (:then_0
          (bind _asbga_keys_and_locks_display_case (Utilities|Casting|CastToBGA_KeysAndLocks_DisplayCase _returnvalue))
          (Class|BGAKeysandLocksDisplayCase|Init _asbga_keys_and_locks_display_case self)
          (Utilities|Array|Add (Variables|Default|GetSpawnedDisplayCases) _asbga_keys_and_locks_display_case))
        (:then_1
          (return _returnvalue))))))

(fn ConstructionScript)
