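; Key-operated lock, derived from the Athena lock parent, that gates vault doors and Keys and Locks display cases, including multi-lock setups.
; Interaction requires and consumes a key item, then one timeline animates the turn and unlocks at its edge event.
; Until all locks open, the key's owner can retrieve it, reversing that timeline.
; Replicated lock state drives colors and key visibility, with net dormancy toggled around animations and quest events sent on completion.
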(event EventHideAndKill
  ;[1 1 1] Override so the lock parent can never tell this to die
  (Development|PrintString "Somoene told me to KMS but I said nooo" false false))

(event EventBeginPlay
  (|CallParentFunction)
  (bind _returnvalue (Rendering|Material|CreateDynamicMaterialInstance (Variables|Default|GetLock_Housing_Main) 0 "/KeysAndLocks/Items/KeysAndLocks/Materials/MI_Gadget_Lock.MI_Gadget_Lock" "None" :!class /Script/Engine.PrimitiveComponent))
  (Variables|Default|SetLockDynamicMat _returnvalue)
  (bind _returnvalue_2 (Rendering|Material|CreateDynamicMaterialInstance (Variables|Default|GetLock) 0 0 "None" :!class /Script/Engine.PrimitiveComponent))
  (bind _output_get (Variables|Default|SetMIDI_ConsoleScreen _returnvalue_2))
  (Utilities|IsValid _output_get
    (:Is%20Valid
      (bind _returnvalue_3 (Rendering|Material|GetVectorParameterValue _output_get "Emissive" :!class /Script/Engine.MaterialInstanceDynamic))
      (Variables|Default|SetOriginalEmissive _returnvalue_3))
    (:Is%20Not%20Valid
      _))
  (bind _returnvalue_4 (Rendering|Material|CreateDynamicMaterialInstance (Variables|Default|GetLock_Housing_Rays) 0 "/KeysAndLocks/Items/KeysAndLocks/Meshes/Materials/MI_KeysAndLocks_Rays.MI_KeysAndLocks_Rays" "None" :!class /Script/Engine.PrimitiveComponent))
  (Variables|Default|SetLockRaysDynamicMat _returnvalue_4)
  (CallFunction|UpdateVisualsBasedonCurrentLockState))

(event Interaction|EventBlueprintOnInteract
  (params (InteractingPawn :type Fort%20Pawn%20Object%20Reference) (InteractionBeingAttempted :type EInteractionBeingAttempted%20Enum :default "FirstInteraction"))
  (if (Variables|Default|GetAbleToTakeKeyBack)
    (CallFunction|TakeKeyBack)
    (else
      (|CallParentFunction InteractingPawn InteractionBeingAttempted))))

(event EventSetLock
  (params (LockState :type Enum_Athena_Lock%20Enum :default "NewEnumerator0"))
  ;[1 1 1] Play success stinger and then open door
  (Variables|LockParentInteract|SetCanFirstInteract false)
  (Utilities|FlowControl|Sequence
    (:then_0
      (Variables|Default|SetPawnWhoOwnsKey (Variables|LockParentInteract|GetLastInteractingPawn))
      (Ability|ExecuteGameplayCueReplicated (Variables|Default|GetGC_Success))
      (Inventory|RemoveItemFromPlayer :PlayerController (Fortnite|GetFortPlayerControllerfromActor (Variables|LockParentInteract|GetLastInteractingPawn)) :ItemDefinition (Variables|Default|GetKeyID) :AmountToRemove 1 :bForceRemoval true :bSkipVerbRemoval false)
      (CallFunction|ShowUnlockEffects))
    (:then_1
      (switch Utilities|FlowControl|Switch|SwitchonEnum_Athena_Lock LockState
        (:NewEnumerator1
          (CallFunction|SetAbletoTakeKeyBack false))))))

(event EventApplyFailCheckEffect
  ;[1 1 1] Play fail stinger - logic for sounding alarms and doing damage, etc would go here
  (CallFunction|MulticastInteractionFailed (Variables|LockParentInteract|GetLastInteractingPawn)))

(event Custom|CallOpenVault
  (bind _asvault_door (Utilities|Casting|CastToVault_Door (Variables|LockParent|GetAttachedLockActor)))
  (Class|VaultDoor|VaultOpen _asvault_door)
  ;[1 1 1] Send quest progress
  (Quest|SendComplexCustomStatEvent (Quests|GetQuestManager (Fortnite|GetFortPlayerControllerfromActor (Variables|LockParentInteract|GetLastInteractingPawn)) "Invalid") _asvault_door "(GameplayTags=((TagName=\"Athena.Quests.UnlockFaction.Door\")))" (GameplayTags|GetOwnedGameplayTags _asvault_door :!class /Script/GameplayTags.GameplayTagAssetInterface) 1))

(event Custom|Multicast_InteractionFailed
  (params (Pawn :type Fort%20Pawn%20Object%20Reference))
  ;[0.8129 0 1] Multicast_InteractionFailed
  ; Plays a fail gameplay cue on the client of the failing player.
  (Utilities|IsValid Pawn
    (:Is%20Valid
      (Utilities|IsValid (Fortnite|GetFortPlayerControllerfromActor Pawn)
        (:Is%20Valid
          (GameplayCue|ExecuteGameplayCue Pawn (Variables|Default|GetGC_Fail)))))))

(event Custom|ShowUnlockEffects
  ;[0.9135 0 1 0.5] Lock Turning animation.
  (Variables|Default|SetPlayingUnlockAnim true)
  (Networking|SetNetDormancy "DORM_Awake")
  (bind (_direction _x_loc _rotation _lockrotation) (Timeline|MoveKeyIn PlayFromStart)
    (:Update
      (Networking|CanExecuteCosmeticEvents
        (:True
          (Transformation|SetRelativeLocation (Variables|Default|GetKey) (Math|Vector|MakeVector (Utilities|RemapFloat _x_loc "(X=0.000000,Y=1.000000)" "(X=47.000000,Y=17.000000)" true) 0.0 2.3) false false)
          (Transformation|SetRelativeRotation (Variables|Default|GetKey) (Math|Rotator|MakeRotator 9e+01 (Utilities|RemapFloat _rotation "(X=0.000000,Y=1.000000)" "(X=90.000000,Y=0.000000)" true) 2.7e+02) false false)
          (Transformation|SetRelativeLocation (Variables|Default|GetLock_Housing_Top) (Math|Vector|MakeVector (Utilities|RemapFloat _rotation "(X=0.000000,Y=1.000000)" "(X=0.000000,Y=-3.000000)" true)) false false)
          (Transformation|SetRelativeRotation (Variables|Default|GetLockPivot) (Math|Rotator|MakeRotator (Utilities|RemapFloat _lockrotation "(X=0.000000,Y=1.000000)" "(X=0.000000,Y=-90.000000)" true)) false false))))
    (:Finished
      ;;[1 0.5777 0] Timeline animation finished. Become dormant
      (Networking|SwitchHasAuthority
        (:Authority
          (Networking|FlushNetDormancy)
          (Networking|SetNetDormancy "DORM_DormantAll"))))
    (:LockUnlockEdge
      (Variables|LockParentInteract|SetCanFirstInteract true)
      (switch Utilities|FlowControl|Switch|SwitchonETimelineDirection _direction
        (:Forward
          ;;[0 0.0615 1] When the lock has been completely unlocked by the key.
          (Variables|LockParent|SetLockState "NewEnumerator2")
          (CallFunction|SetAbletoTakeKeyBack (not (CallFunction|AllLocksUnlocked)))
          (Variables|Default|SetPlayingUnlockAnim false)
          (Variables|Default|SetSimpleLockState "NewEnumerator1")
          (CallFunction|UpdateVisualsBasedonCurrentLockState)
          (|CallParentFunction) ; Call the parent OnUnlock function to unlock. The non-parent version does nothing so that outside forces don't force an unlock.
          (Networking|CanExecuteCosmeticEvents
            (:True
              (Niagara|SpawnSystematLocation "/KeysAndLocks/Items/KeysAndLocks/FX/NS_KeysLocks_LockOpen.NS_KeysLocks_LockOpen" (Transformation|GetWorldLocation (Variables|Default|GetUnlockVFX))))))
        (:Backward
          (Variables|LockParent|SetLockState "NewEnumerator1")
          (Variables|Interaction|SetbAllowInteract true)
          (Variables|Default|SetSimpleLockState "NewEnumerator0")
          (CallFunction|SetAbletoTakeKeyBack false)
          (Networking|SwitchHasAuthority
            (:Authority
              ;;[0.3646 0.3448 0] The lock has just been re-locked on the server.
              (InventoryManagement|AddItem (Fortnite|GetFortPlayerControllerfromActor (Variables|Default|GetPawnWhoOwnsKey)) "/KeysAndLocks/Items/KeysAndLocks/AGID_KeysAndLocks_Key.AGID_KeysAndLocks_Key" 1)
              (Variables|LockParent|SetLockState "NewEnumerator1")
              (Ability|ExecuteGameplayCueReplicated (Variables|Default|GetGC_TakeKeyBack)))
            (:Remote
              ;;[0.401 0.1497 0] The lock has just been re-locked on the client
              (Rendering|SetVisibility (Variables|Default|GetKey) false false)
              (CallFunction|UpdateVisualsBasedonCurrentLockState))))))))

(event Custom|TakeKeyBack
  ;[0.9135 0 1 0.5] Lock Turning animation.
  (Networking|SetNetDormancy "DORM_Awake")
  (Timeline|MoveKeyIn ReverseFromEnd))

(event EventMultilockCompleted
  (bind _pawnwhoownskey (Variables|Default|GetPawnWhoOwnsKey))
  (bind _returnvalue (Fortnite|GetFortPlayerControllerfromActor _pawnwhoownskey))
  (Default|CallOnMultilockCompleted self)
  (Utilities|FlowControl|Sequence
    (:then_0
      (CallFunction|SetAbletoTakeKeyBack false)
      (Utilities|IsValid _returnvalue
        (:Is%20Valid
          (Quest|SendComplexCustomStatEvent (Quests|GetQuestManager _returnvalue "Invalid") 0 (Variables|Default|GetQuestUnlockTag) "(GameplayTags=)" 1))))
    (:then_1
      (Variables|Default|SetSimpleLockState "NewEnumerator2")
      (Networking|FlushNetDormancy)
      (Networking|SetNetDormancy "DORM_DormantAll"))))

(event EventOpenContainers
  (bind _returnvalue (Actor|GetAttachParentActor))
  (|CallParentFunction)
  (bind _returnvalue_2 (Niagara|SpawnSystematLocation "/KeysAndLocks/Items/KeysAndLocks/FX/NS_KeysLocks_VaultOpen.NS_KeysLocks_VaultOpen" (+ (Transformation|GetWorldLocation (Variables|Default|GetUnlockVFX)) (* (Transformation|GetForwardVector (Variables|Default|GetUnlockVFX)) -2e+01))))
  (bind _asbga_keys_and_locks_display_case (Utilities|Casting|CastToBGA_KeysAndLocks_DisplayCase _returnvalue)
    (:then
      (Effects|Components|ParticleSystem|SetColorParameter _returnvalue_2 "Color" (Variables|/KeysAndLocks/Items/KeysAndLocks/BGA_KeysAndLocks_DisplayCase.BGA_KeysAndLocks_DisplayCase_C|GetFXRarityColor _asbga_keys_and_locks_display_case)))
    (:CastFailed
      (bind _asbga_keys_and_locks_display_case_2locks (Utilities|Casting|CastToBGA_KeysAndLocks_DisplayCase_2Locks _returnvalue))
      (Effects|Components|ParticleSystem|SetColorParameter _returnvalue_2 "Color" (Variables|/KeysAndLocks/Items/KeysAndLocks/BGA_KeysAndLocks_DisplayCase.BGA_KeysAndLocks_DisplayCase_C|GetFXRarityColor _asbga_keys_and_locks_display_case_2locks)))))

(event Custom|UpdateColors
  (params (State :type Enum_KeysAndLocks_LockState%20Enum :default "NewEnumerator0"))
  (bind _returnvalue (Utilities|Select :Index State :NewEnumerator0 "(R=1.000000,G=0.032086,B=0.000000,A=1.000000)" :NewEnumerator1 "(R=0.053686,G=1.000000,B=0.000000,A=1.000000)" :NewEnumerator2 "(R=0.053686,G=1.000000,B=0.000000,A=1.000000)"))
  (Networking|CanExecuteCosmeticEvents
    (:True
      (Rendering|Material|SetMaterial (Variables|Default|GetLock_Housing_Main) 0 (Utilities|Select :Index State :NewEnumerator0 "/KeysAndLocks/Items/KeysAndLocks/Materials/MI_Gadget_Lock.MI_Gadget_Lock" :NewEnumerator1 "/KeysAndLocks/Items/KeysAndLocks/Materials/TEMP_MI_Gadget_Lock_Green.TEMP_MI_Gadget_Lock_Green" :NewEnumerator2 "/KeysAndLocks/Items/KeysAndLocks/Materials/TEMP_MI_Gadget_Lock_Green.TEMP_MI_Gadget_Lock_Green"))
      (Rendering|Material|SetVectorParameterValue (Variables|Default|GetLockRaysDynamicMat) "EmissionColor" _returnvalue :!class /Script/Engine.MaterialInstanceDynamic)
      (Rendering|Components|Light|SetLightColor (Variables|Default|GetLock_Light) (* _returnvalue (Utilities|Select :Index State :NewEnumerator0 "(R=1.000000,G=1.000000,B=1.000000,A=1.000000)" :NewEnumerator1 "(R=0.070000,G=0.070000,B=0.070000,A=1.000000)" :NewEnumerator2 "(R=0.070000,G=0.070000,B=0.070000,A=1.000000)")) true))))

(event EventOnUnlocked)

(event Custom|CallParentOnUnlocked
  (|CallParentFunction))

(fn BlueprintCanInteract
  (params (InteractingPawn :type Fort%20Pawn%20Object%20Reference) (InteractionBeingAttempted :type EInteractionBeingAttempted%20Enum :default "FirstInteraction") (InteractionType :type TInteractionType%20Enum :default "IT_NoInteraction"))
  (for _array_element (Variables|LockMultipleLocks|GetOtherLocksOnActor)
    (bind _asbga_keys_and_locks_lock (Utilities|Casting|CastToBGA_KeysAndLocks_Lock _array_element))
    (LocalVariables|SetIsAnimating (Variables|Default|GetPlayingUnlockAnim _asbga_keys_and_locks_lock)))
  (return (and (or (Math|Boolean|ANDBoolean (Variables|Default|GetAbleToTakeKeyBack) (== (Variables|Default|GetPawnWhoOwnsKey) InteractingPawn) (not (Variables|Default|GetPlayingUnlockAnim)) (not (LocalVariables|GetIsAnimating))) (and (|CallParentFunction InteractingPawn InteractionBeingAttempted InteractionType) (Utilities|Enum|Equal(Enum) (Variables|LockParent|GetLockState) "NewEnumerator1"))) (not (Variables|Default|GetPlayingUnlockAnim)))))

(fn BlueprintGetFailedSubTextTexture
  (params (InteractingPawn :type Fort%20Pawn%20Object%20Reference) (InteractionBeingAttempted :type EInteractionBeingAttempted%20Enum :default "FirstInteraction"))
  (return (Variables|Default|GetFailedInteractTexture)))

(fn BlueprintGetInteractionString
  (params (InteractingPawn :type Fort%20Pawn%20Object%20Reference) (InteractionBeingAttempted :type EInteractionBeingAttempted%20Enum :default "FirstInteraction"))
  (|CallParentFunction InteractingPawn InteractionBeingAttempted)
  (if (Variables|Default|GetAbleToTakeKeyBack)
    (return (Variables|Default|GetInteractionText_TakeKey))
    (else
      (bind _checkpassed (CallFunction|LockCheck self InteractingPawn InteractionBeingAttempted))
      (if _checkpassed
        (return (Variables|Default|GetInteractionText_HaveKey))
        (else
          (return (Variables|Default|GetInteractionText_NeedKey)))))))

(fn BlueprintGetInteractionTime
  (params (InteractingPawn :type Fort%20Pawn%20Object%20Reference) (InteractionBeingAttempted :type EInteractionBeingAttempted%20Enum :default "FirstInteraction"))
  (bind (_outinteractiontime _returnvalue) (|CallParentFunction InteractingPawn InteractionBeingAttempted))
  (bind _returnvalue_2 (CallFunction|ShouldUseLongInteract InteractingPawn))
  (return (Math|Float|SelectFloat 0.75 0.0 _returnvalue_2) _returnvalue))

(fn BlueprintGetSubInteractionString
  (params (InteractingPawn :type Fort%20Pawn%20Object%20Reference) (InteractionBeingAttempted :type EInteractionBeingAttempted%20Enum :default "FirstInteraction"))
  (|CallParentFunction InteractingPawn InteractionBeingAttempted)
  (if (Variables|Default|GetAbleToTakeKeyBack)
    (return "")
    (else
      (bind _asbga_athena_lock_parent (Utilities|Casting|CastToBGA_Athena_Lock_Parent self)) ; This guy is a liar, we do need to cast to the parent
      (bind _checkpassed (CallFunction|LockCheck _asbga_athena_lock_parent InteractingPawn "FirstInteraction"))
      (if _checkpassed
        (return "")
        (else
          (return (Variables|Default|GetInteractionText_NeedKey_Secondary)))))))

(fn BlueprintOnLocalInteract
  (params (InteractingPawn :type Fort%20Player%20Pawn%20Object%20Reference) (InteractionBeingAttempted :type EInteractionBeingAttempted%20Enum :default "FirstInteraction"))
  (bind _returnvalue (Ability|GetFortAbilitySystemComponent InteractingPawn))
  (bind _returnvalue_2 (|CallParentFunction InteractingPawn InteractionBeingAttempted))
  (Utilities|FlowControl|Sequence
    (:then_0
      (Utilities|IsValid _returnvalue
        (:Is%20Valid
          (bind _returnvalue_3 (CallFunction|ShouldUseLongInteract InteractingPawn))
          (Abilities|TryActivateAbilitybyClass _returnvalue (select _returnvalue_3 "/Game/Abilities/Player/Generic/Traits/DefaultPlayer/GA_DefaultPlayer_InteractSearch.GA_DefaultPlayer_InteractSearch_C" "/Game/Abilities/Player/Generic/Traits/DefaultPlayer/GA_DefaultPlayer_InteractUse.GA_DefaultPlayer_InteractUse_C") false))))
    (:then_1
      (return _returnvalue_2))))

(fn BlueprintUpdateDynamicInteractionOptions
  (params (InteractingPawn :type Fort%20Pawn%20Object%20Reference) (InteractionBeingAttempted :type EInteractionBeingAttempted%20Enum :default "FirstInteraction") (InteractionType :type TInteractionType%20Enum :default "IT_NoInteraction"))
  (bind _outgameplaytags (|CallParentFunction InteractingPawn InteractionBeingAttempted InteractionType))
  (LocalVariables|SetInteractionOptions _outgameplaytags)
  (bind _asbga_athena_lock_parent (Utilities|Casting|CastToBGA_Athena_Lock_Parent self)
    (:then
      (bind _checkpassed (CallFunction|LockCheck _asbga_athena_lock_parent InteractingPawn InteractionBeingAttempted))
      (if (not _checkpassed)
        (GameplayTags|AddGameplayTag (LocalVariables|GetInteractionOptions) "(TagName=\"HUD.Athena.InteractionOptions.DisableInteractKey\")" :!class /Script/GameplayTags.BlueprintGameplayTagLibrary)))
    (:CastFailed
      _)) ; This guy is a liar, we do need to cast to the parent
  (return (LocalVariables|GetInteractionOptions)))

(fn LockCheck
  (params (InteractingPawn :type Fort%20Pawn%20Object%20Reference) (InteractionBeingAttempted :type EInteractionBeingAttempted%20Enum :default "FirstInteraction"))
  (return (> (Inventory|GetItemQuantityOnPlayer (Fortnite|GetFortPlayerControllerfromActor InteractingPawn) "/KeysAndLocks/Items/KeysAndLocks/AGID_KeysAndLocks_Key.AGID_KeysAndLocks_Key") 0)))

(fn OnRep_PlayingUnlockAnim
  (if (Variables|Default|GetPlayingUnlockAnim)
    (Rendering|SetVisibility (Variables|Default|GetKey) true false)))

(fn OnRep_SimpleLockState
  (CallFunction|UpdateVisualsBasedonCurrentLockState))

(fn SetAbleToTakeKeyBack
  (params (AbleToTakeKeyBack :type Boolean))
  (bind _self self)
  (bind _takekeybacktag (Variables|Default|GetTakeKeyBackTag))
  (bind _output_get (Variables|Default|SetAbleToTakeKeyBack AbleToTakeKeyBack))
  (if _output_get
    (Ability|GameplayEffect|AddLooseGameplayTags _self _takekeybacktag true)
    (else
      (Ability|GameplayEffect|RemoveLooseGameplayTags _self _takekeybacktag true))))

(fn ShouldUseLongInteract
  (params (InteractingPawn :type Fort%20Pawn%20Object%20Reference))
  (return (and (> (Inventory|GetItemQuantityOnPlayer (Fortnite|GetFortPlayerControllerfromActor InteractingPawn) "/KeysAndLocks/Items/KeysAndLocks/AGID_KeysAndLocks_Key.AGID_KeysAndLocks_Key") 0) (not (Variables|Default|GetAbleToTakeKeyBack)))))

(fn UpdateVisualsBasedOnCurrentLockState
  (bind _lockpivot (Variables|Default|GetLockPivot))
  (bind _lock_housing_top (Variables|Default|GetLock_Housing_Top))
  (bind _key (Variables|Default|GetKey))
  (Networking|CanExecuteCosmeticEvents
    (:True
      (bind _output_get (LocalVariables|SetVisualLockState (select (Variables|Default|GetPlayingUnlockAnim) "NewEnumerator1" (Variables|Default|GetSimpleLockState))))
      (CallFunction|UpdateColors _output_get)
      (switch Utilities|FlowControl|Switch|SwitchonEnum_KeysAndLocks_LockState _output_get
        (:NewEnumerator0
          (Rendering|SetVisibility _key false false)
          (Transformation|SetRelativeLocation _lock_housing_top (0.0 0.0 0.0) false false)
          (Transformation|SetRelativeRotation _lockpivot (0.0 0.0 0.0) false false))
        (:NewEnumerator1
          (Rendering|SetVisibility _key true false))
        (:NewEnumerator2
          (Rendering|SetVisibility _key false false)))
      (Transformation|SetRelativeLocation _lock_housing_top (-3.0 0.0 0.0) false false)
      (Transformation|SetRelativeRotation _lockpivot (0.0 0.0 -9e+01) false false)
      (Transformation|SetRelativeLocation _key (17.0 0.0 2.3) false false)
      (Transformation|SetRelativeRotation _key (0.0 2.7e+02 9e+01) false false))))

(fn ConstructionScript
  (|CallParentFunction))
