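; Vault lock actor from Keys and Locks that shows a client-side marker to a local player who nears it while holding a key.
; Trigger overlap on clients checks the key-equipped tag, opened state and existing markers, then registers one and stores its ID.
; UnmarkLock clears it when the pawn leaves or loses the key tag.
; Multilock completion destroys the AthenaSpecialActor component, sets Opened and unmarks all local players.
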(event EventBeginPlay
  (|CallParentFunction) ; This node is disabled and will not be called.\nDrag off pins to build functionality.
)

(event Collision|EventActorBeginOverlap
  (params (OtherActor :type Actor%20Object%20Reference)))

(event EventTick
  (params (DeltaSeconds :type Float%20%28single-precision%29)))

(event EventMultilockCompleted
  (|CallParentFunction)
  (Components|DestroyComponent (Variables|Default|GetAthenaSpecialActor))
  (Variables|Default|SetOpened true)
  ;[1 0 0.8826 0.5] Now that the vault has opened up, remove the ping marker from all slient player pawns.
  (Networking|CanExecuteCosmeticEvents
    (:True
      (for _array_element (Fortnite|GetLocalFortPlayerControllers true)
        (Utilities|IsValid _array_element
          (:Is%20Valid
            (CallFunction|UnmarkLock (Game|Player|GetPlayerPawn _array_element))))))))

(event OnComponentBeginOverlap(PlayerTriggerForPing_)
  (params (OverlappedComponent :type Primitive%20Component%20Object%20Reference) (OtherActor :type Actor%20Object%20Reference) (OtherComp :type Primitive%20Component%20Object%20Reference) (OtherBodyIndex :type Integer) (bFromSweep :type Boolean) (SweepResult :type Hit%20Result%20Structure%20%28by%20ref%29))
  (Networking|SwitchHasAuthority
    (:Remote
      (bind _returnvalue (GameplayTags|HasMatchingGameplayTag(Message) OtherActor "(TagName=\"Granted.ItemEquipped.Key\")"))
      (if (and _returnvalue (not (Variables|Default|GetOpened)))
        (bind _aspawn (Utilities|Casting|CastToPawn OtherActor))
        (if (Pawn|IsLocallyControlled _aspawn)
          (bind _asfort_player_controller_athena (Utilities|Casting|CastToFortPlayerControllerAthena (Pawn|GetController _aspawn)))
          (if (Utilities|IsValid (Class|FortPlayerController|GetMarkerComponent _asfort_player_controller_athena) :!class /Script/Engine.KismetSystemLibrary)
            (Utilities|IsValid (MarkerComponent|FindMarkerbyMarkedActor (Class|FortPlayerController|GetMarkerComponent _asfort_player_controller_athena) self)
              (:Is%20Not%20Valid
                (bind (_playerbankid _bankinstanceid) (Fortnite|RegisterSoundIndicatorPlacementContext _asfort_player_controller_athena self "0, 0, 0" false))
                (Variables|SetMarkerID_PlayerID (Utilities|Struct|MakeMarkerID _playerbankid _bankinstanceid))
                (Ability|Tasks|WaitGameplayTagRemoveFromActor OtherActor "(TagName=\"Granted.ItemEquipped.Key\")"
                  (:Removed
                    (CallFunction|UnmarkLock OtherActor)))))))))))

(event OnComponentEndOverlap(PlayerTriggerForPing_)
  (params (OverlappedComponent :type Primitive%20Component%20Object%20Reference) (OtherActor :type Actor%20Object%20Reference) (OtherComp :type Primitive%20Component%20Object%20Reference) (OtherBodyIndex :type Integer))
  (Networking|SwitchHasAuthority
    (:Remote
      (CallFunction|UnmarkLock OtherActor))))

(fn UnmarkLock
  (params (Pawn :type Object%20Reference))
  (bind _aspawn (Utilities|Casting|CastToPawn Pawn))
  (if (Pawn|IsLocallyControlled _aspawn)
    (bind _asfort_player_controller_athena (Utilities|Casting|CastToFortPlayerControllerAthena (Pawn|GetController _aspawn)))
    (Utilities|IsValid (Class|FortPlayerController|GetMarkerComponent _asfort_player_controller_athena)
      (:Is%20Valid
        (MarkerComponent|UnmarkActoronClient (Class|FortPlayerController|GetMarkerComponent _asfort_player_controller_athena) (Variables|Default|GetMarkerID))))))

(fn ConstructionScript
  (|CallParentFunction))
