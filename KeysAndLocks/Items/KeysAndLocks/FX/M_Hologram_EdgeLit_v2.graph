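# Unlit alpha-composite hologram material with edge lighting and panning scanlines, usable on skeletal meshes.
# Fresnel blended with a view-space directional light term combines with scanlines, mapped from pre-skinned position, to shade a two-color tint.
# Desaturated scene color drives dark/light-background brightness and opacity blends, replaced by a constant on mobile.
# An object-bounds sphere mask weights center versus edge before the emissive adjustment function.
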Material /KeysAndLocks/Items/KeysAndLocks/FX/M_Hologram_EdgeLit_v2
graph {
 properties {
  bAllowNegativeEmissiveColor=true
  bUsedWithMeshDeformer=true
  bUsedWithSkeletalMesh=true
  blendMode=BLEND_AlphaComposite
  opacityMaskClipValue=0.1585
  pixelDepthOffsetMode=PDOM_Legacy
  shadingModel=MSM_Unlit
  usedShadingModels=Unlit
 }
 nodes [
  n1 = SceneColor!(UVs)
  n2 = Constant() {r=1}
  n3 = ShadingPathSwitch(
   Default=Clamp(None=Desaturation(None=@n1.RGB, Fraction=@n2), Max=ScalarParameter() {defaultValue=0.9, parameterName=Max Opaqueness})
   Mobile=Constant() {r=1}
  ) {desc=Migrated from FeatureLevelSwitch}
  LinearInterpolate(A=ScalarParameter() {parameterName="Opacity (Dark BG)"}, B=ScalarParameter() {defaultValue=1, parameterName="Opacity (Light BG)"}, Alpha=@n3) {constA=1.5}
  n4 = LinearInterpolate(
   A=ScalarParameter() {defaultValue=1, parameterName="Brightness (Dark BG)"}
   B=ScalarParameter() {defaultValue=5, parameterName="Brightness (Light BG)"}
   Alpha=@n3
  )
  n5 = ComponentMask(None=PreSkinnedPosition()) {b=true, g=true}
  n6 = Panner(
   Coordinate=Divide(A=VertexInterpolator(VS=@n5).PS) {constB=30}
   Time=Time()
   Speed=AppendVector(A=Constant(), B=ScalarParameter() {defaultValue=-0.5, parameterName=Scanline Pan Speed, sliderMax=5, sliderMin=-5})
  ) {speedY=-1}
  n7 = VertexNormalWS()
  n8 = Fresnel(Normal=@n7)
  n9 = Normalize(VectorInput=VectorParameter() {defaultValue=<rgba 0.576 0.2472 0.4977 0>}.RGB)
  n10 = ConstantBiasScale(None=DotProduct(A=@n7, B=Transform(None=@n9) {transformSourceType=TRANSFORMSOURCE_View})) {scale=0.468}
  n11 = ScalarParameter() {defaultValue=0.016, parameterName=Lighting Mix, sliderMax=1}
  n12 = Power(
   Base=Multiply(A=LinearInterpolate(A=@n8, B=@n10, Alpha=@n11), B=ScalarParameter() {defaultValue=1, parameterName=Lighting Weight, sliderMax=1})
   Exp=ScalarParameter() {defaultValue=1.44, parameterName=Lighting Mix Exponent, sliderMax=12}
  ) {constExponent=3}
  n13 = Multiply(
   A=Multiply(
    A=TextureSampleParameter2D(UVs=@n6) {parameterName=Param, texture=/Game/PvP/StrategicBuildingActors/Defensive/LaserFence/Assets/LaserScanline}.RGB
    B=ScalarParameter() {defaultValue=0.096, parameterName=ScanlineWeight, sliderMax=1}
   )
   B=OneMinus(None=@n12)
  )
  n14 = Multiply(
   A=LinearInterpolate(
    A=VectorParameter() {parameterName=Color A, defaultValue=<rgba 0 0.006 0.5365 1>}.RGB
    B=VectorParameter() {parameterName=Color B, defaultValue=<rgba 0.042 0.3096 1 1>}.RGB
    Alpha=NamedRerouteUsage()
   )
   B=LinearInterpolate(A=Fresnel() {baseReflectFraction=0.3, exponent=6}, Alpha=Add(A=@n13, B=@n12)) {constA=0.3}
  )
  n15 = ScalarParameter() {defaultValue=0.8, parameterName="Opacity Multiplier (Edge)"}
  n16 = ScalarParameter() {defaultValue=1, parameterName="Opacity Multiplier (Center)"}
  n17 = NamedRerouteDeclaration(None=SphereMask(
    A=ObjectPositionWS()
    B=WorldPosition().XYZ
    Radius=ObjectRadius()
    Hardness=ScalarParameter() {defaultValue=0.1, parameterName=Hardness, sliderMax=1}
   ) {hardnessPercent=0.1}) {name=Sphere Mask, nodeColor=<rgba 1 0.7969 0 1>}
  MaterialFunctionCall(Emissive=Multiply(A=Multiply(A=@n4, B=@n14), B=LinearInterpolate(A=@n15, B=@n16, Alpha=@n17) {constA=0.2, constB=0.95})) {materialFunction=&MF_FX_EmissiveAdjustments}
 ]
}
