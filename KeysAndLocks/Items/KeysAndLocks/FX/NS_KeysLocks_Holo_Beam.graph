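# Niagara system for the Keys and Locks holographic beam effect, tinted by a shared system color.
# GPU mesh emitters layer blocky beams, a cone and short streaks, sized and spread mostly by user Beam Height and Width.
# A pulsing, rotating sprite carries a light with user radius, intensity and offset.
# An entry emitter spawns expanding, fading blocky sprites at a system-computed rate; an alternate beam emitter is disabled.
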NiagaraSystem /KeysAndLocks/Items/KeysAndLocks/FX/NS_KeysLocks_Holo_Beam
dependencies {
 usedDataInterfaces [&NiagaraDataInterfaceCurve, &NiagaraDataInterfaceColorCurve]
 usedDynamicInputs [&ClampFloat, &Abs_Float, &Subtract_Float]
 usedModules [
  &self:SystemSpawnScript.SystemScriptSource.SystemScriptGraph.NiagaraNodeAssignment_4.SetVariables_#1
  &SystemState
  &self:SystemSpawnScript.SystemScriptSource.SystemScriptGraph.NiagaraNodeAssignment_0.SetVariables_#2
  &self:SystemSpawnScript.SystemScriptSource.SystemScriptGraph.NiagaraNodeAssignment_1.SetVariables_#3
  &self:SystemSpawnScript.SystemScriptSource.SystemScriptGraph.NiagaraNodeAssignment_3.SetVariables_#4
 ]
}
emitters {
 Beam_blocky001 {
  simTarget=GPUComputeSim
  emitterUpdateScript [
   &EmitterState()
   &SpawnRate(SpawnRate=2, Use Spawn Probability=false)
  ]
  particleSpawnScript [
   &InitializeParticle(
    Color=&MultiplyLinearColor(Linear Color=$System.SysColor, Scale Factor=<0.6 0.6 0.6 1>)
    Lifetime Max=0.8
    Lifetime Min=0.2
    Lifetime Mode=::NewEnumerator1
    Mesh Scale=&Multiply_VectorByFloat(Vector=<2 1 1>, Float=$User.Beam Height)
    Mesh Scale Mode=::NewEnumerator4
    Position Offset=&MakeVector(X=0, Y=0, Z=$System.BaseHeight)
    UsePositionOffset=true
    Write Material Random=true
   )
   &DynamicMaterialParameters(
    Index 0 Param 1=3
    Index 0 Param 2=&RandomRangeFloat(Minimum=0.3, Maximum=0.7)
    Index 0 Param 3=0.25
    Index 0 Param 4=1
    Module.Param0WriteEnabled=true
    Module.Param1WriteEnabled=true
    Module.Param2WriteEnabled=true
    Module.Param3WriteEnabled=true
   )
   /Niagara/Modules/Update/Orientation/V2/OrientMeshToVector(
    Look At Direction:Vector3f=&Add_Vector(
     A=<1 0 0>
     B=&RandomVector(Vector Scale=&Multiply_Float(A=0.5, B=$User.Beam Width))
    )
    Override Quat=false
    UseOptionalSideAxis=false
    UseOptionalUpAxis=false
   )
  ]
  particleUpdateScript [
   &ParticleState()
   &ScaleColor(
    Scale Alpha=&FloatFromCurve(
     FloatCurve=DataInterface() {propertyValues {Curve {keys [
         {interpMode=RCIM_Cubic, tangentMode=RCTM_User, time=0, value=0}
         {interpMode=RCIM_Cubic, tangentMode=RCTM_User, time=0.5, value=1}
         {tangentMode=RCTM_User, time=1, value=0}
        ]}}}
     Scale Curve=0.1
    )
    ScaleA=true
    ScaleRGB=false
   )
   &DynamicMaterialParameters(
    Index 0 Param 2=&FloatFromCurve(FloatCurve=DataInterface() {propertyValues {Curve {keys [
         {arriveTangent=5.1604, interpMode=RCIM_Cubic, leaveTangent=5.1604, tangentMode=RCTM_User, time=0, value=0}
         {interpMode=RCIM_Cubic, tangentMode=RCTM_User, time=0.5, value=1}
         {tangentMode=RCTM_User, time=1, value=0}
        ]}}})
    Module.Param0WriteEnabled=false
    Module.Param1WriteEnabled=true
    Module.Param2WriteEnabled=false
    Module.Param3WriteEnabled=false
   ) {module=DynamicMaterialParameters001}
   &CameraOffset(Camera Offset Amount=15) {enabled=false}
  ]
  rendererClasses [&NiagaraMeshRendererProperties]
  renderers [{rendererClass=&NiagaraMeshRendererProperties, rendererIndex=0}]
 }
 Beam_blocky_cone {
  simTarget=GPUComputeSim
  emitterUpdateScript [
   &EmitterState()
   &SpawnRate(Use Spawn Probability=false)
  ]
  particleSpawnScript [
   &InitializeParticle(
    Color=&MultiplyLinearColor(Linear Color=$System.SysColor, Scale Factor=<0.15 0.15 0.15 1>)
    Lifetime Max=1.2
    Lifetime Min=0.75
    Lifetime Mode=::NewEnumerator1
    Mesh Scale=&Multiply_VectorByFloat(
     Vector=<5 5 2>
     Float=&Multiply_Float(A=&RandomRangeFloat(Minimum=0.75, Maximum=1), B=$User.Beam Height)
    )
    Mesh Scale Mode=::NewEnumerator4
    Position Offset=&MakeVector(X=0, Y=0, Z=0)
    UsePositionOffset=true
    Write Material Random=true
   )
   &DynamicMaterialParameters(
    Index 0 Param 1=3
    Index 0 Param 2=&RandomRangeFloat(Minimum=0.3, Maximum=0.7)
    Index 0 Param 3=0.25
    Index 0 Param 4=1
    Module.Param0WriteEnabled=true
    Module.Param1WriteEnabled=true
    Module.Param2WriteEnabled=true
    Module.Param3WriteEnabled=true
   )
   &OrientMeshToVector(Look At Direction:Vector3f=<0 0 -1>, Override Quat=false, UseOptionalSideAxis=false, UseOptionalUpAxis=false)
  ]
  particleUpdateScript [
   &ParticleState()
   &ScaleColor(
    Scale Alpha=&FloatFromCurve(
     FloatCurve=DataInterface() {propertyValues {Curve {keys [
         {arriveTangent=5.1262, interpMode=RCIM_Cubic, leaveTangent=5.1262, tangentMode=RCTM_User, time=0, value=0}
         {interpMode=RCIM_Cubic, tangentMode=RCTM_User, time=0.5, value=1}
         {tangentMode=RCTM_User, time=1, value=0}
        ]}}}
     Scale Curve=0.5
    )
    ScaleA=true
    ScaleRGB=false
   )
   &DynamicMaterialParameters(
    Index 0 Param 1=0.75
    Index 0 Param 2=3
    Index 2 Param 1=0.5
    Index 2 Param 2=0.5
    Index 3 Param 1=&RandomRangeFloat(Minimum=0, Maximum=1)
    Module.Param0WriteEnabled=true
    Module.Param0WriteEnabled2=true
    Module.Param0WriteEnabled3=true
    Module.Param1WriteEnabled=true
    Module.Param1WriteEnabled2=true
    Module.Param1WriteEnabled3=false
    Module.Param2WriteEnabled=false
    Module.Param2WriteEnabled2=false
    Module.Param2WriteEnabled3=false
    Module.Param3WriteEnabled=false
    Module.Param3WriteEnabled2=false
    Module.Param3WriteEnabled3=false
    Write Parameter Index 2=true
    Write Parameter Index 3=true
   ) {module=DynamicMaterialParameters002}
   &CameraOffset(Camera Offset Amount=15) {enabled=false}
  ]
  rendererClasses [&NiagaraMeshRendererProperties]
  renderers [{rendererClass=&NiagaraMeshRendererProperties, rendererIndex=0}]
 }
 Beams001 {
  bEnabled=false
  simTarget=GPUComputeSim
  emitterUpdateScript [
   &EmitterState()
   &SpawnRate(SpawnRate=3, Use Spawn Probability=false)
  ]
  particleSpawnScript [
   &InitializeParticle(
    Color=$System.SysColor
    Lifetime Max=1.5
    Lifetime Min=0.5
    Lifetime Mode=::NewEnumerator1
    Mesh Scale Mode=::NewEnumerator3
    Mesh Uniform Scale=&Multiply_Float(A=&RandomRangeFloat(Minimum=0.75, Maximum=1.5), B=$User.Beam Height)
    Position Offset=&MakeVector(X=0, Y=0, Z=0)
    UsePositionOffset=true
    Write Material Random=true
   )
   &DynamicMaterialParameters(
    Index 0 Param 1=1.5
    Index 0 Param 2=&RandomRangeFloat(Minimum=0.3, Maximum=0.7)
    Index 0 Param 3=0.25
    Index 0 Param 4=1
    Module.Param0WriteEnabled=true
    Module.Param1WriteEnabled=true
    Module.Param2WriteEnabled=true
    Module.Param3WriteEnabled=true
   )
   &OrientMeshToVector(
    Look At Direction:Vector3f=&Add_Vector(
     A=<1 0 0>
     B=&RandomVector(Vector Scale=&Multiply_Float(A=0.15, B=$User.Beam Width))
    )
    Override Quat=false
    UseOptionalSideAxis=false
    UseOptionalUpAxis=false
   )
  ]
  particleUpdateScript [
   &ParticleState()
   &ScaleColor(
    Scale Alpha=&FloatFromCurve(
     FloatCurve=DataInterface() {propertyValues {Curve {keys [
         {arriveTangent=4.8167, interpMode=RCIM_Cubic, leaveTangent=4.8167, tangentMode=RCTM_User, time=0, value=0}
         {interpMode=RCIM_Cubic, tangentMode=RCTM_User, time=0.5, value=1}
         {tangentMode=RCTM_User, time=1, value=0}
        ]}}}
     Scale Curve=0.3
    )
    ScaleA=true
    ScaleRGB=false
   )
   &DynamicMaterialParameters(
    Index 0 Param 2=&FloatFromCurve(FloatCurve=DataInterface() {propertyValues {Curve {keys [
         {arriveTangent=2.4641, interpMode=RCIM_Cubic, leaveTangent=2.4641, tangentMode=RCTM_User, time=0, value=0.3}
         {interpMode=RCIM_Cubic, tangentMode=RCTM_User, time=0.5, value=1}
         {tangentMode=RCTM_User, time=1, value=0.3}
        ]}}})
    Module.Param0WriteEnabled=false
    Module.Param1WriteEnabled=true
    Module.Param2WriteEnabled=false
    Module.Param3WriteEnabled=false
   ) {module=DynamicMaterialParameters001}
   &CameraOffset(Camera Offset Amount=15) {enabled=false}
  ]
  rendererClasses [&NiagaraMeshRendererProperties]
  renderers [
   {rendererClass=&NiagaraMeshRendererProperties, rendererIndex=0}
   {rendererClass=&NiagaraMeshRendererProperties, rendererIndex=1}
  ]
 }
 BlockyStuff_onEnter {
  simTarget=GPUComputeSim
  emitterUpdateScript [
   &EmitterState(Inactive Response=::NewEnumerator1, Life Cycle Mode=::NewEnumerator1, Loop Duration=1.5, UseLoopDelay=false)
   &SpawnRate(SpawnRate=$System.OnEntry_Rate, Use Spawn Probability=false)
  ]
  particleSpawnScript [
   &InitializeParticle(
    Color=$System.SysColor
    Lifetime=1.3
    Sprite Size Mode=::NewEnumerator3
    Uniform Sprite Size=50
    UsePositionOffset=false
    Write Material Random=true
   )
   &AddVelocity(Velocity=<55 0 0>)
   &ApplyInitialForces()
   /SetVariables(Particles.Position=&Add_Vector(A=Linked() {linkedVariable {name=Engine.Owner.Position, type=NiagaraPosition}}, B=<0 0 15>)) {isSetParameters=true}
  ]
  particleUpdateScript [
   &ParticleState()
   /SetVariables(Particles.SpriteSize=&Multiply_Vector2D_ByFloat(
     Float=&FloatFromCurve(
      FloatCurve=DataInterface() {propertyValues {Curve {keys [
          {arriveTangent=5.8161, interpMode=RCIM_Cubic, leaveTangent=5.8161, tangentMode=RCTM_User, time=0, value=1}
          {arriveTangent=4.5838, interpMode=RCIM_Cubic, leaveTangent=4.5838, tangentMode=RCTM_User, time=0.3, value=2.625}
          {arriveTangent=-0.0405, interpMode=RCIM_Cubic, leaveTangent=-0.0405, tangentMode=RCTM_User, time=1, value=4.375}
         ]}}}
      Scale Curve=15
     )
     Vector2D=<1 1>
    )) {isSetParameters=true}
   &ScaleColor(
    Scale Alpha=&FloatFromCurve(
     FloatCurve=DataInterface() {propertyValues {Curve {keys [
         {arriveTangent=-0.2933, interpMode=RCIM_Cubic, leaveTangent=-0.2933, tangentMode=RCTM_User, time=0, value=0}
         {arriveTangent=-2.8313, interpMode=RCIM_Cubic, leaveTangent=-2.8313, tangentMode=RCTM_User, time=0.1, value=0.95}
         {arriveTangent=-0.9489, interpMode=RCIM_Cubic, leaveTangent=-0.9489, tangentMode=RCTM_User, time=0.2, value=0.25}
         {arriveTangent=-0.0383, interpMode=RCIM_Cubic, leaveTangent=-0.0383, tangentMode=RCTM_User, time=1, value=0}
        ]}}}
     Scale Curve=0.75
    )
    Scale RGB=&Multiply_VectorByFloat(
     Vector=<1 1 1>
     Float=&FloatFromCurve(
      FloatCurve=DataInterface() {propertyValues {Curve {keys [
          {arriveTangent=-0.2933, interpMode=RCIM_Cubic, leaveTangent=-0.2933, leaveTangentWeight=0.3333, tangentMode=RCTM_User, time=0, value=0}
          {
           arriveTangent=0.0608
           arriveTangentWeight=0.0333
           interpMode=RCIM_Cubic
           leaveTangent=0.0608
           leaveTangentWeight=0.3
           tangentMode=RCTM_User
           time=0.0594
           value=4.75
          }
          {arriveTangent=-27.9094, interpMode=RCIM_Cubic, leaveTangent=-27.9094, tangentMode=RCTM_User, time=0.085, value=4.0956}
          {
           arriveTangent=-0.6853
           arriveTangentWeight=0.0333
           interpMode=RCIM_Cubic
           leaveTangent=-0.6853
           leaveTangentWeight=0.2667
           tangentMode=RCTM_User
           time=0.5036
           value=0.3433
          }
          {arriveTangent=-0.8945, arriveTangentWeight=0.2667, interpMode=RCIM_Cubic, leaveTangent=-0.8945, tangentMode=RCTM_User, time=1, value=0}
         ]}}}
      Scale Curve=0.45
     )
    )
    ScaleA=true
    ScaleRGB=true
   ) {module=ScaleColor001}
   &SolveForcesAndVelocity(Clamp Velocity=false, Limit Acceleration=false)
   &DynamicMaterialParameters(
    Index 0 Param 1=0.75
    Index 0 Param 2=4
    Index 2 Param 1=0.55
    Index 2 Param 2=0.4
    Index 3 Param 1=&RandomRangeFloat(Minimum=0, Maximum=1)
    Module.Param0WriteEnabled=true
    Module.Param0WriteEnabled2=true
    Module.Param0WriteEnabled3=true
    Module.Param1WriteEnabled=true
    Module.Param1WriteEnabled2=true
    Module.Param1WriteEnabled3=false
    Module.Param2WriteEnabled=true
    Module.Param2WriteEnabled2=false
    Module.Param2WriteEnabled3=false
    Module.Param3WriteEnabled=true
    Module.Param3WriteEnabled2=false
    Module.Param3WriteEnabled3=false
    Write Parameter Index 2=true
    Write Parameter Index 3=true
   )
   &CameraOffset(Camera Offset Amount=25) {enabled=false, module=CameraOffset001}
  ]
  rendererClasses [&NiagaraSpriteRendererProperties]
  renderers [{rendererClass=&NiagaraSpriteRendererProperties}]
 }
 SingleLoopingParticle001 {
  simTarget=GPUComputeSim
  emitterUpdateScript [
   &EmitterState()
   &SpawnBurst_Instantaneous(Spawn Count=1, Use Spawn Probability=false)
  ]
  particleSpawnScript [
   &InitializeParticle(
    Color=&HueShiftLinearColor(Hue=0, Saturation=1.5, Value=0.5, Linear Color=$System.SysColor)
    Lifetime=$Emitter.CurrentLoopDuration
    Sprite Size Mode=::NewEnumerator3
    Uniform Sprite Size=135
    UsePositionOffset=false
    Write Material Random=true
   )
   /SetVariables(Particles.SpriteSize_B:Vector2f=&Multiply_Vector2D_ByFloat(Float=0.25, Vector2D=$Particles.SpriteSize)) {isSetParameters=true}
  ]
  particleUpdateScript [
   &ParticleState()
   &SolveForcesAndVelocity(Clamp Velocity=false, Limit Acceleration=false)
   &SpriteRotationRate(Rotation Rate=&Lerp_Float(A=4, B=3, Alpha=&Sine(Normalized Angle=$Emitter.Age, Period=3, Scale=5, Bias=3)))
   &ScaleSpriteSize(
    Uniform Curve Index=&Sine(Normalized Angle=$Emitter.Age, Period=12, Scale=1, Bias=0)
    Uniform Curve Sprite Scale=DataInterface() {propertyValues {Curve {
       postInfinityExtrap=RCCE_Cycle
       keys [
        {arriveTangent=-0.1152, interpMode=RCIM_Cubic, leaveTangent=-0.1152, tangentMode=RCTM_User, time=0, value=0.625}
        {arriveTangent=-2.5249, interpMode=RCIM_Cubic, leaveTangent=-2.5249, tangentMode=RCTM_User, time=0.3, value=1}
        {arriveTangent=-0.0405, interpMode=RCIM_Cubic, leaveTangent=-0.0405, tangentMode=RCTM_User, time=0.7, value=0.625}
       ]
      }}}
   )
   /SetVariables(Particles.Position=&AddVectorToPosition(Position=<0 0 0>, Vector=<0 0 0>)) {isSetParameters=true}
   &self:ScaleSpriteBySystem_01() {module=ScaleSpriteBySystem_01}
   &FN_Light_Attributes(
    Color=$Particles.Color
    Create Color Attribute=true
    Create Radius Attribute=true
    Radius=$User.Light Radius
    Scale Color=$User.Light Intensity
   )
   /SetVariables(Particles.LightPosition:NiagaraPosition=&MakeVector(X=$User.Light Offset, Y=0, Z=0)) {isSetParameters=true}
   &CameraOffset(Camera Offset Amount=15)
  ]
  rendererClasses [&NiagaraSpriteRendererProperties, &NiagaraLightRendererProperties]
  renderers [
   {rendererClass=&NiagaraSpriteRendererProperties}
   {rendererClass=&NiagaraLightRendererProperties, rendererIndex=1}
  ]
 }
 Streaks001 {
  simTarget=GPUComputeSim
  emitterUpdateScript [
   &EmitterState()
   &SpawnRate(SpawnRate=12, Use Spawn Probability=false)
  ]
  particleSpawnScript [
   &InitializeParticle(
    Color=$System.SysColor
    Lifetime=&RandomRangeFloat(Minimum=0.2, Maximum=0.7)
    Mesh Scale Max=<1 0.2 0.2>
    Mesh Scale Min=<1 0.1 0.05>
    Mesh Scale Mode=::NewEnumerator2
    UsePositionOffset=false
    Write Material Random=true
   )
   &DynamicMaterialParameters(
    Index 0 Param 1=0.1
    Index 0 Param 2=&RandomRangeFloat(Minimum=0.2, Maximum=0.1)
    Index 0 Param 3=1
    Index 0 Param 4=5
    Module.Param0WriteEnabled=true
    Module.Param1WriteEnabled=true
    Module.Param2WriteEnabled=true
    Module.Param3WriteEnabled=true
   )
   &OrientMeshToVector(
    Look At Direction:Vector3f=&Add_Vector(A=<1 0 0>, B=&RandomVector(Vector Scale=0.3))
    Override Quat=false
    UseOptionalSideAxis=false
    UseOptionalUpAxis=false
   )
  ]
  particleUpdateScript [
   &ParticleState()
   &ScaleColor(
    Scale Mode=::NewEnumerator1
    Scale RGBA=&Vector4FromCurve(Vector4Curve=DataInterface() {propertyValues {
       AlphaCurve {keys [<at 0.0129 0>,<at 0.4401 0.3>,<at 1 0>]}
       BlueCurve {keys [<at 0.0037 0>,<at 0.4309 0.4586>,<at 1 1>]}
       GreenCurve {keys [<at 0.0037 0>,<at 0.4309 0.4635>,<at 1 0.9571>]}
       RedCurve {keys [<at 0.0037 0>,<at 0.4309 0.4844>,<at 1 0.9886>]}
      }})
    ScaleRGBA=true
   )
  ]
  rendererClasses [&NiagaraMeshRendererProperties]
  renderers [
   {rendererClass=&NiagaraMeshRendererProperties, rendererIndex=0}
   {rendererClass=&NiagaraMeshRendererProperties, rendererIndex=1}
   {rendererClass=&NiagaraMeshRendererProperties, rendererIndex=2}
  ]
 }
}
summary {
 systemName=NS_KeysLocks_Holo_Beam
 emitters [
  {bEnabled=false, emitterName=Beams001, simTarget=GPUComputeSim, rendererClasses [&NiagaraMeshRendererProperties]}
  {
   emitterName=SingleLoopingParticle001
   simTarget=GPUComputeSim
   rendererClasses [&NiagaraSpriteRendererProperties, &NiagaraLightRendererProperties]
  }
  {emitterName=Beam_blocky001, simTarget=GPUComputeSim, rendererClasses [&NiagaraMeshRendererProperties]}
  {emitterName=Streaks001, simTarget=GPUComputeSim, rendererClasses [&NiagaraMeshRendererProperties]}
  {emitterName=BlockyStuff_onEnter, simTarget=GPUComputeSim, rendererClasses [&NiagaraSpriteRendererProperties]}
  {emitterName=Beam_blocky_cone, simTarget=GPUComputeSim, rendererClasses [&NiagaraMeshRendererProperties]}
 ]
 userVariables [
  User.Beam Color {type=LinearColor, description="", defaultValue {struct=LinearColor, value=<rgba 0.0981 0.2844 0.6302 2>}}
  User.Beam Height {type=NiagaraFloat, defaultValue=0.5, description=""}
  User.Beam Width {type=NiagaraFloat, defaultValue=0.5, description=""}
  User.EntrySignal {type=NiagaraFloat, defaultValue=1, description=""}
  User.Light Intensity {type=NiagaraFloat, defaultValue=3, description=""}
  User.Light Offset {type=NiagaraFloat, defaultValue=50, description=""}
  User.Light Radius {type=NiagaraFloat, defaultValue=1.2, description=""}
 ]
}
