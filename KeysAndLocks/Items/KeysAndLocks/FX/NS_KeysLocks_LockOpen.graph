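# Niagara burst effect for KeysAndLocks, named for a lock opening: short-lived glow sprites, an expanding halo and light flashes.
# Halo_and_Glow and single_halo001 grow and fade camera-offset sprites through curves, while LightRenderer fires two timed light bursts.
# All three emitters tint from System.SysColor; a User.Color parameter is exposed.
# The single halo also scales alpha by time of day and modulates RGB brightness with NPC music amplitude.
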NiagaraSystem /KeysAndLocks/Items/KeysAndLocks/FX/NS_KeysLocks_LockOpen
dependencies {
 usedDataInterfaces [&NiagaraDataInterfaceCurve, &NiagaraDataInterfaceColorCurve]
 usedDynamicInputs []
 usedModules [
  &SystemState
  &self:SystemSpawnScript.SystemScriptSource.SystemScriptGraph.NiagaraNodeAssignment_0.SetVariables_#1
 ]
}
emitters {
 Halo_and_Glow {
  emitterUpdateScript [
   &EmitterState()
   &SpawnRate(Use Spawn Probability=false) {enabled=false}
   &SpawnBurst_Instantaneous(Spawn Count=1, Use Spawn Probability=false)
  ]
  particleSpawnScript [&InitializeParticle(Color=$System.SysColor, Lifetime=0.5, UsePositionOffset=true, Write Material Random=true)]
  particleUpdateScript [
   &ParticleState()
   &Color(
    Color=&ColorFromCurve(ColorCurve=DataInterface() {propertyValues {
       AlphaCurve {keys [<at 0.1791 1 RCIM_Cubic>,<at 0.9745 0 RCIM_Cubic>]}
       BlueCurve {keys [<at 0.2314 0.3642>, {arriveTangent=-0.9419, interpMode=RCIM_Cubic, leaveTangent=-0.9419, time=0.6814, value=0}]}
       GreenCurve {keys [<at 0.2314 3>, {arriveTangent=-5.563, interpMode=RCIM_Cubic, leaveTangent=-5.563, time=0.6814, value=0.849}]}
       RedCurve {keys [<at 0.2314 0>, {arriveTangent=1.3256, interpMode=RCIM_Cubic, leaveTangent=1.3256, time=0.6814, value=0.5126}]}
      }})
    Scale Color=<0.05 0.05 0.05>
   ) {enabled=false}
   &ScaleColor(
    Scale Alpha=&FloatFromCurve(
     FloatCurve=DataInterface() {propertyValues {Curve {keys [
         {arriveTangent=10.4544, interpMode=RCIM_Cubic, leaveTangent=10.4544, tangentMode=RCTM_User, time=0, value=0}
         {interpMode=RCIM_Cubic, tangentMode=RCTM_User, time=0.3, value=0.95}
         {tangentMode=RCTM_User, time=1, value=0}
        ]}}}
     Scale Curve=0.3
    )
    ScaleA=true
    ScaleRGB=false
   ) {module=ScaleColor001}
   /SetVariables(Particles.SpriteSize=&Vector2DFromFloat(Value=&FloatFromCurve(
      FloatCurve=DataInterface() {propertyValues {Curve {keys [
          {arriveTangent=0.3463, interpMode=RCIM_Cubic, leaveTangent=0.3463, tangentMode=RCTM_User, time=0, value=0.2}
          {arriveTangent=1.8357, interpMode=RCIM_Cubic, leaveTangent=1.8357, tangentMode=RCTM_User, time=0.3, value=0.95}
          {arriveTangent=-0.0405, interpMode=RCIM_Cubic, leaveTangent=-0.0405, tangentMode=RCTM_User, time=1, value=1.5}
         ]}}}
      Scale Curve=55
     ))) {isSetParameters=true}
   &CameraOffset(Camera Offset Amount=55)
   &FN_Radial_Material(
    Radial Mask Power=1
    Radial Mask Tile U=0.2
    Radial Mask Tile V=0.5
    Sprite Extrusion Speed:NiagaraFloat=&FloatFromCurve(
     FloatCurve=DataInterface() {propertyValues {Curve {keys [
         {arriveTangent=-0.2256, interpMode=RCIM_Cubic, time=0, value=1.75}
         {arriveTangent=2.75, interpMode=RCIM_Cubic, leaveTangent=2.75, time=0.3, value=2.375}
         {arriveTangent=2.5, interpMode=RCIM_Cubic, leaveTangent=2.5, time=0.5, value=3.125}
         {interpMode=RCIM_Cubic, leaveTangent=-0.5666, time=0.7, value=3.375}
        ]}}}
     Scale Curve=-0.3
    )
    Sprite Spin Speed:NiagaraFloat=&FloatFromCurve(FloatCurve=DataInterface() {propertyValues {Curve {keys [<at 0 0.2>,<at 0.4 1.125>,<at 1 2.25>]}}}, Scale Curve=0.1)
   ) {enabled=false}
  ]
  rendererClasses [&NiagaraSpriteRendererProperties]
  renderers [
   {rendererClass=&NiagaraSpriteRendererProperties}
   {rendererClass=&NiagaraSpriteRendererProperties, rendererIndex=1}
   {rendererClass=&NiagaraSpriteRendererProperties, rendererIndex=2}
  ]
 }
 LightRenderer {
  emitterUpdateScript [
   &EmitterState(Life Cycle Mode=::NewEnumerator1, Loop Behavior=::NewEnumerator1, Loop Duration=2, UseLoopDelay=false)
   &SpawnBurst_Instantaneous(Use Spawn Probability=false, UseLoopCountLimit=false)
   &SpawnBurst_Instantaneous(Spawn Time=1.2327, Use Spawn Probability=false, UseLoopCountLimit=false) {module=SpawnBurst_Instantaneous001}
  ]
  particleSpawnScript [
   &InitializeParticle(
    Color=$System.SysColor
    Lifetime=0.5
    Sprite Size Mode=::NewEnumerator1
    Uniform Sprite Size Max=100
    Uniform Sprite Size Min=55
    UsePositionOffset=false
    Write Material Random=true
   )
   &SphereLocation(
    MaskSpawnGroup=false
    Override Seed=false
    Sphere Radius=2
    Surface Only=false
    UseNonUniformScale=false
    UseOffset=true
    UseSphereOrientation=false
   )
   /SetVariables(Particles.LightRadius=100, Particles.LightScale:Vector3f=<1 1 1>) {isSetParameters=true}
  ]
  particleUpdateScript [
   &ParticleState()
   &ScaleColor(
    Scale Alpha=&FloatFromCurve(FloatCurve=DataInterface() {propertyValues {Curve {keys [<at 0.2 2>,<at 0.7 0.5>,<at 1 0>]}}}, Scale Curve=0.3)
    ScaleA=true
    ScaleRGB=true
   )
   &CameraOffset(Camera Offset Amount=55)
  ]
  rendererClasses [&NiagaraLightRendererProperties]
  renderers [{rendererClass=&NiagaraLightRendererProperties, rendererIndex=0}]
 }
 single_halo001 {
  emitterUpdateScript [
   &EmitterState(Inactive Response=::NewEnumerator1, Life Cycle Mode=::NewEnumerator1, Loop Behavior=::NewEnumerator1, Loop Duration=1.5, UseLoopDelay=false)
   &SpawnBurst_Instantaneous(Use Spawn Probability=false, UseLoopCountLimit=false)
  ]
  particleSpawnScript [&InitializeParticle(
    Color=$System.SysColor
    Lifetime=0.3
    Sprite Size Mode=::NewEnumerator3
    Uniform Sprite Size=75
    UsePositionOffset=false
    Write Material Random=true
   )]
  particleUpdateScript [
   &ParticleState()
   &CameraOffset(Camera Offset Amount=15)
   /SetVariables(Particles.SpriteSize=&Multiply_Vector2D_ByFloat(
     Float=&FloatFromCurve(
      FloatCurve=DataInterface() {propertyValues {Curve {keys [
          {arriveTangent=-0.0094, interpMode=RCIM_Cubic, leaveTangent=-0.0094, tangentMode=RCTM_User, time=0, value=0}
          {arriveTangent=7.6747, interpMode=RCIM_Cubic, leaveTangent=7.6747, tangentMode=RCTM_User, time=0.3, value=2}
          {arriveTangent=-0.0405, interpMode=RCIM_Cubic, leaveTangent=-0.0405, tangentMode=RCTM_User, time=1, value=4.375}
         ]}}}
      Scale Curve=50
     )
     Vector2D=<1 1>
    )) {isSetParameters=true}
   &ScaleColor(
    Scale Alpha=&FloatFromCurve(
     FloatCurve=DataInterface() {propertyValues {Curve {keys [
         {arriveTangent=-0.2933, interpMode=RCIM_Cubic, leaveTangent=-0.2933, tangentMode=RCTM_User, time=0, value=0}
         {arriveTangent=-2.8313, interpMode=RCIM_Cubic, leaveTangent=-2.8313, tangentMode=RCTM_User, time=0.1, value=0.95}
         {arriveTangent=-1.6622, interpMode=RCIM_Cubic, leaveTangent=-1.6622, tangentMode=RCTM_User, time=0.2, value=0.45}
         {arriveTangent=-0.0383, interpMode=RCIM_Cubic, leaveTangent=-0.0383, tangentMode=RCTM_User, time=1, value=0}
        ]}}}
     Scale Curve=0.3
    )
    ScaleA=true
    ScaleRGB=false
   ) {module=ScaleColor001}
   &ScaleColor(
    Scale Alpha=&Multiply_Float(
     A=&FloatFromCurve(FloatCurve=DataInterface() {propertyValues {Curve {keys [<at 0.8 1>,<at 1 0>]}}})
     B=&FNTimeOfDayScalar(Night Scale=0.3, Morning Scale=0.1, Evening Scale=0.2)
    )
    Scale RGB=&VectorFromFloat(Value=&FloatFromCurve(
      FloatCurve=DataInterface() {propertyValues {Curve {keys [<at 0 1>,<at 0.1 3>,<at 0.2 1>]}}}
      Scale Curve=&Lerp_Float(A=1, B=3, Alpha=$NPC.NPC_Audio_Reactive_Parameters.Amplitude_Music)
     ))
    ScaleA=true
    ScaleRGB=true
   )
  ]
  rendererClasses [&NiagaraSpriteRendererProperties]
  renderers [
   {rendererClass=&NiagaraSpriteRendererProperties}
   {rendererClass=&NiagaraSpriteRendererProperties, rendererIndex=1}
  ]
 }
}
summary {
 systemName=NS_KeysLocks_LockOpen
 emitters [
  {emitterName=LightRenderer, rendererClasses [&NiagaraLightRendererProperties]}
  {emitterName=Halo_and_Glow, rendererClasses [&NiagaraSpriteRendererProperties]}
  {emitterName=single_halo001, rendererClasses [&NiagaraSpriteRendererProperties]}
 ]
 userVariables [User.Color {type=LinearColor, description="", defaultValue {struct=LinearColor, value=<rgba 0.1843 0.3229 0 1>}}]
}
