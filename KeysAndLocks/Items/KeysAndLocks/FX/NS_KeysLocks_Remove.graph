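# Niagara one-shot effect for removal in the KeysAndLocks item set, reading as a brief inward-collapsing burst of tinted sprites.
# A glow fades while a halo shrinks, GPU radial flares contract on a torus, and curl-noise sparkles fly toward the center.
# Halo color comes from User.Color, other layers use System.SysColor, and sparkle alpha scales with time of day.
# Sprites are X-offset and pushed toward the camera; the debug gnomon is disabled.
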NiagaraSystem /KeysAndLocks/Items/KeysAndLocks/FX/NS_KeysLocks_Remove
dependencies {
 usedDataInterfaces [&NiagaraDataInterfaceCurve, &NiagaraDataInterfaceColorCurve, &NiagaraDataInterfaceVector2DCurve]
 usedDynamicInputs [&MultiplyLinearColor, &MakeVector4_FromFloat]
 usedModules [
  &SystemState
  &self:SystemSpawnScript.SystemScriptSource.SystemScriptGraph.NiagaraNodeAssignment_0.SetVariables_#1
 ]
}
emitters {
 FN_DebugGnomon {
  bEnabled=false
  emitterUpdateScript [
   &EmitterState(Life Cycle Mode=::NewEnumerator1, Loop Behavior=::NewEnumerator1, Loop Duration=1, UseLoopDelay=false)
   &SpawnBurst_Instantaneous(Use Spawn Probability=false, UseLoopCountLimit=false)
  ]
  particleSpawnScript [
   &InitializeParticle(
    Color Mode=::NewEnumerator0
    Lifetime=5
    Mesh Scale Mode=::NewEnumerator3
    Mesh Uniform Scale=5
    UsePositionOffset=false
    Write Material Random=true
   )
   /SetVariables(Particles.Position=&AddVectorToPosition(Position=$Particles.Position, Vector=<15 0 0>)) {isSetParameters=true}
  ]
  particleUpdateScript [&ParticleState()]
  rendererClasses [&NiagaraMeshRendererProperties]
  renderers [{rendererClass=&NiagaraMeshRendererProperties, rendererIndex=0}]
 }
 NE_GlowDecal {
  emitterUpdateScript [
   &EmitterState(Life Cycle Mode=::NewEnumerator1, Loop Behavior=::NewEnumerator1, Loop Duration=1, UseLoopDelay=false)
   &SpawnBurst_Instantaneous(Spawn Count=1, Use Spawn Probability=false)
  ]
  particleSpawnScript [
   /Niagara/Modules/Spawn/Initialization/InitializeParticle(
    Color=$System.SysColor
    Lifetime=0.35
    Module.Write Color=true
    Module.Write Lifetime=true
    Module.Write Mass=true
    Module.Write Material Random=true
    Module.Write Position=false
    Module.Write Scale=false
    Module.Write SpriteRotation=false
    Module.Write SpriteSize=true
    Position=&SimulationPosition()
    Sprite Size=<125 125>
   )
   &AddVelocity() {enabled=false}
   &SphereLocation(
    MaskSpawnGroup=false
    Override Seed=false
    Sphere Radius=0
    Surface Only=false
    UseNonUniformScale=false
    UseOffset=false
    UseSphereOrientation=false
   )
   /SetVariables(Particles.Position=&AddVectorToPosition(Position=$Particles.Position, Vector=<15 0 0>)) {isSetParameters=true}
  ]
  particleUpdateScript [
   &ParticleState()
   &GravityForce()
   &SolveForcesAndVelocity(Clamp Velocity=false, Limit Acceleration=false)
   &Color(Color=$User.Color) {enabled=false}
   &ScaleColor(
    Scale Alpha=&FloatFromCurve(FloatCurve=DataInterface() {propertyValues {Curve {keys [
         {tangentMode=RCTM_User, time=0, value=1}
         {arriveTangent=-0.1769, interpMode=RCIM_Cubic, leaveTangent=-0.1769, tangentMode=RCTM_User, time=0.4, value=1.05}
         {arriveTangent=-5.3315, interpMode=RCIM_Cubic, leaveTangent=-5.3315, tangentMode=RCTM_User, time=1, value=0}
        ]}}})
    ScaleA=true
    ScaleRGB=false
   )
   &CameraOffset(Camera Offset Amount=125) {module=CameraOffset001}
  ]
  rendererClasses [&NiagaraSpriteRendererProperties]
  renderers [{rendererClass=&NiagaraSpriteRendererProperties}]
 }
 radial_flares {
  simTarget=GPUComputeSim
  emitterUpdateScript [
   &EmitterState(Inactive Response=::NewEnumerator1, Life Cycle Mode=::NewEnumerator1, Loop Behavior=::NewEnumerator1, Loop Duration=0.3, UseLoopDelay=false)
   &SpawnBurst_Instantaneous(Spawn Count=12, Use Spawn Probability=false, UseLoopCountLimit=false)
  ]
  particleSpawnScript [/Niagara/Modules/Spawn/Initialization/V2/InitializeParticle(Color=$System.SysColor, Sprite Size=<0.6 12>, Sprite Size Mode=::NewEnumerator4, UsePositionOffset=false, Write Material Random=true)]
  particleUpdateScript [
   &ParticleState()
   &TorusLocation(
    Align Torus With Axis=true
    Handle Radius=0
    Large Radius:NiagaraFloat=&FloatFromCurve(
     FloatCurve=DataInterface() {propertyValues {Curve {keys [
         {arriveTangent=-40.4056, interpMode=RCIM_Cubic, leaveTangentWeight=0.3879, time=0, value=0.2}
         {arriveTangent=9.9281, interpMode=RCIM_Cubic, leaveTangent=9.9281, tangentMode=RCTM_User, time=0.1, value=0.55}
         {arriveTangentWeight=0.2335, interpMode=RCIM_Cubic, leaveTangent=0.36, time=0.1381, value=1}
        ]}}}
     Scale Curve=&RandomRangeFloat(Minimum=9, Maximum=9)
    )
    MaskSpawnGroup=false
    Normalized Angle Around Torus Axis=1
    Torus Axis=<0 1 0>
    Torus Distribution Mode:ENiagaraTorusDistributionMode=::NewEnumerator1
    U Position:NiagaraFloat=&Add_Float(A=&ReturnNormalizedExecIndex(), B=0)
   ) {enabled=false}
   &TorusLocation(
    Align Torus With Axis=true
    Handle Radius=0
    Large Radius:NiagaraFloat=&FloatFromCurve(
     FloatCurve=DataInterface() {propertyValues {Curve {keys [
         {arriveTangent=-1.0862, interpMode=RCIM_Cubic, leaveTangentWeight=0.3879, time=0, value=1}
         {arriveTangent=-9.1916, interpMode=RCIM_Cubic, leaveTangent=-9.1916, tangentMode=RCTM_User, time=0.0441, value=0.4752}
         {arriveTangentWeight=0.2335, interpMode=RCIM_Cubic, leaveTangent=0.36, time=0.1381, value=0.2}
        ]}}}
     Scale Curve=&RandomRangeFloat(Minimum=3, Maximum=12)
    )
    MaskSpawnGroup=false
    Normalized Angle Around Torus Axis=1
    Torus Axis=<0 1 0>
    Torus Distribution Mode:ENiagaraTorusDistributionMode=::NewEnumerator1
    U Position:NiagaraFloat=&Add_Float(A=&ReturnNormalizedExecIndex(), B=0)
   ) {module=TorusLocation002}
   &self:ScratchModule_03() {module=ScratchModule_03}
   &ScaleColor(
    Scale Alpha=&FloatFromCurve(
     FloatCurve=DataInterface() {propertyValues {Curve {keys [
         {arriveTangent=-0.2933, interpMode=RCIM_Cubic, leaveTangent=-0.2933, tangentMode=RCTM_User, time=0.2, value=0}
         {arriveTangent=2.4658, interpMode=RCIM_Cubic, leaveTangent=2.4658, tangentMode=RCTM_User, time=0.7, value=0.9}
         {arriveTangent=0.0301, interpMode=RCIM_Cubic, leaveTangent=0.0301, tangentMode=RCTM_User, time=0.9, value=0.1}
        ]}}}
     Scale Curve=3
    )
    ScaleA=true
    ScaleRGB=false
   ) {module=ScaleColor001}
   &ScaleSpriteSize(
    Uniform Curve Scale=3
    Uniform Curve Sprite Scale=DataInterface() {propertyValues {Curve {keys [
        {arriveTangent=-0.2933, interpMode=RCIM_Cubic, leaveTangent=-0.2933, leaveTangentWeight=0.4283, tangentMode=RCTM_User, time=0, value=1.35}
        {
         arriveTangent=-4.5651
         arriveTangentWeight=0.1274
         interpMode=RCIM_Cubic
         leaveTangent=-4.5651
         leaveTangentWeight=0.7778
         tangentMode=RCTM_User
         time=0.1721
         value=0.757
        }
        {arriveTangent=0.0301, arriveTangentWeight=0.2866, interpMode=RCIM_Cubic, leaveTangent=0.0301, tangentMode=RCTM_User, time=0.7254, value=0}
       ]}}}
   )
   &CameraOffset(Camera Offset Amount=125)
   /SetVariables(Particles.Position=&AddVectorToPosition(Position=$Particles.Position, Vector=<15 0 0>)) {isSetParameters=true}
  ]
  rendererClasses [&NiagaraSpriteRendererProperties]
  renderers [{rendererClass=&NiagaraSpriteRendererProperties}]
 }
 single_halo001 {
  emitterUpdateScript [
   &EmitterState(Inactive Response=::NewEnumerator1, Life Cycle Mode=::NewEnumerator1, Loop Behavior=::NewEnumerator1, Loop Duration=1.5, UseLoopDelay=false)
   &SpawnBurst_Instantaneous(Use Spawn Probability=false, UseLoopCountLimit=false)
  ]
  particleSpawnScript [
   &InitializeParticle(Color=$User.Color, Sprite Size Mode=::NewEnumerator3, Uniform Sprite Size=1, UsePositionOffset=false, Write Material Random=true)
   /SetVariables(Particles.Position=&AddVectorToPosition(Position=$Particles.Position, Vector=<15 0 0>)) {isSetParameters=true}
  ]
  particleUpdateScript [
   &ParticleState()
   &CameraOffset(Camera Offset Amount=125)
   /SetVariables(Particles.SpriteSize=&Multiply_Vector2D_ByFloat(
     Float=&FloatFromCurve(
      FloatCurve=DataInterface() {propertyValues {Curve {keys [
          {arriveTangent=-6.9622, interpMode=RCIM_Cubic, time=0, value=4.25}
          {arriveTangent=-4.5689, interpMode=RCIM_Cubic, leaveTangent=-4.5689, time=0.3, value=3.5}
          {arriveTangent=-0.6878, interpMode=RCIM_Cubic, leaveTangent=-0.6878, tangentMode=RCTM_User, time=0.9576, value=-0.125}
         ]}}}
      Scale Curve=15
     )
     Vector2D=<1 1>
    )) {isSetParameters=true}
   &ScaleColor(
    Scale Alpha=&FloatFromCurve(
     FloatCurve=DataInterface() {propertyValues {Curve {keys [
         {arriveTangent=-0.2933, interpMode=RCIM_Cubic, leaveTangent=-0.2933, tangentMode=RCTM_User, time=0, value=0}
         {arriveTangent=-2.8313, interpMode=RCIM_Cubic, leaveTangent=-2.8313, tangentMode=RCTM_User, time=0.1, value=0.95}
         {arriveTangent=-1.6622, interpMode=RCIM_Cubic, leaveTangent=-1.6622, tangentMode=RCTM_User, time=0.2, value=0.45}
         {arriveTangent=-0.0383, interpMode=RCIM_Cubic, leaveTangent=-0.0383, tangentMode=RCTM_User, time=1, value=0}
        ]}}}
     Scale Curve=0.25
    )
    ScaleA=true
    ScaleRGB=false
   ) {module=ScaleColor001}
  ]
  rendererClasses [&NiagaraSpriteRendererProperties]
  renderers [
   {rendererClass=&NiagaraSpriteRendererProperties}
   {rendererClass=&NiagaraSpriteRendererProperties, rendererIndex=1}
  ]
 }
 sparkle_center {
  emitterUpdateScript [
   &EmitterState(Inactive Response=::NewEnumerator1, Life Cycle Mode=::NewEnumerator1, Loop Behavior=::NewEnumerator1, Loop Duration=2, UseLoopDelay=false)
   &SpawnRate(SpawnRate=50, Use Spawn Probability=false) {enabled=false}
   &SpawnBurst_Instantaneous(Spawn Count=40, Use Spawn Probability=false, UseLoopCountLimit=false)
  ]
  particleSpawnScript [
   &InitializeParticle(
    Color=$System.SysColor
    Lifetime=&RandomRangeFloat(Minimum=0.3, Maximum=0.75)
    Sprite Rotation Mode=::NewEnumerator2
    Sprite Size Mode=::NewEnumerator1
    Uniform Sprite Size Max=6
    Uniform Sprite Size Min=3.5
    UsePositionOffset=false
    Write Material Random=true
   )
   /SetVariables(Particles.RandomCore:NiagaraFloat=&FloatFromCurve(
     FloatCurve=DataInterface() {propertyValues {Curve {keys [
         {arriveTangent=0.0559, interpMode=RCIM_Cubic, leaveTangent=0.0559, tangentMode=RCTM_User, time=0, value=0.2}
         {arriveTangent=0.1895, interpMode=RCIM_Cubic, leaveTangent=0.1895, tangentMode=RCTM_User, time=0.9, value=0.25}
         {arriveTangent=1.1808, interpMode=RCIM_Cubic, leaveTangent=1.1808, tangentMode=RCTM_User, time=1, value=0.3}
        ]}}}
     CurveIndex=&RandomRangeFloat(Minimum=0, Maximum=1)
     Scale Curve=3
    )) {isSetParameters=true}
   &SphereLocation(
    MaskSpawnGroup=false
    Override Seed=false
    Sphere Radius=125
    Surface Only=false
    UseNonUniformScale=false
    UseOffset=false
    UseSphereOrientation=false
   )
   /SetVariables(Particles.Position=&AddVectorToPosition(Position=$Particles.Position, Vector=<15 0 0>)) {isSetParameters=true}
   &AddVelocityFromPoint(Module.Use Falloff Distance=false, Velocity Strength=&RandomRangeFloat(Minimum=-200, Maximum=-300))
  ]
  particleUpdateScript [
   &ParticleState()
   &CurlNoiseForce(Noise Frequency=125, Noise Strength=&Multiply_Float(A=555, B=1), Pan Noise=true, Pan Noise Field=<0.5 0.5 0>)
   &ScaleSpriteSize(Scale Factor=&Vector2DFromFloat(Value=&FloatFromCurve(
      FloatCurve=DataInterface() {propertyValues {Curve {keys [
          <at 0 1 RCIM_Cubic>
          {arriveTangent=-1, interpMode=RCIM_Cubic, leaveTangent=-1, time=0.6, value=1.25}
          {arriveTangent=-0.1423, interpMode=RCIM_Cubic, leaveTangent=-0.1423, tangentMode=RCTM_User, time=1, value=0}
         ]}}}
      Scale Curve=$Particles.RandomCore
     )))
   &ScaleSpriteSize(Scale Factor=&Multiply_Vector2D(A=$Particles.SpriteSize, B=<0.27 1>), Scale Sprite Size Mode=::NewEnumerator2) {module=ScaleSpriteSize001}
   &ScaleColor(
    Scale Alpha=&Multiply_Float(
     A=&FloatFromCurve(FloatCurve=DataInterface() {propertyValues {Curve {keys [<at 0 0>,<at 0.1 0.95>,<at 0.6 0.95>,<at 1 0>]}}}, Scale Curve=2)
     B=&FNTimeOfDayScalar(Night Scale=0.3, Morning Scale=0.1, Evening Scale=0.2)
    )
    Scale RGB=&VectorFromFloat(Value=&FloatFromCurve(FloatCurve=DataInterface() {propertyValues {Curve {keys [
          {arriveTangent=9.0725, interpMode=RCIM_Cubic, leaveTangent=9.0725, tangentMode=RCTM_User, time=0, value=0}
          {arriveTangent=-2.5249, interpMode=RCIM_Cubic, leaveTangent=-2.5249, tangentMode=RCTM_User, time=0.1408, value=0.9847}
          {arriveTangent=-0.0405, interpMode=RCIM_Cubic, leaveTangent=-0.0405, tangentMode=RCTM_User, time=1, value=0}
         ]}}}))
    ScaleA=true
    ScaleRGB=true
   )
   &self:ScratchModule_03() {module=ScratchModule_03}
   &Drag(Drag=&RandomRangeFloat(Minimum=0.7, Maximum=1), Use Linear Drag=true, Use Rotational Drag=false)
   &SolveForcesAndVelocity(Clamp Velocity=false, Limit Acceleration=false)
   &CameraOffset(Camera Offset Amount=125)
  ]
  rendererClasses [&NiagaraSpriteRendererProperties]
  renderers [{rendererClass=&NiagaraSpriteRendererProperties}]
 }
}
summary {
 systemName=NS_KeysLocks_Remove
 emitters [
  {emitterName=NE_GlowDecal, rendererClasses [&NiagaraSpriteRendererProperties]}
  {emitterName=single_halo001, rendererClasses [&NiagaraSpriteRendererProperties]}
  {emitterName=radial_flares, simTarget=GPUComputeSim, rendererClasses [&NiagaraSpriteRendererProperties]}
  {emitterName=sparkle_center, rendererClasses [&NiagaraSpriteRendererProperties]}
  {bEnabled=false, emitterName=FN_DebugGnomon, rendererClasses [&NiagaraMeshRendererProperties]}
 ]
 userVariables [
  User.bIsFrontEndPreview {type=NiagaraBool}
  User.Color {type=LinearColor, description="", defaultValue {struct=LinearColor, value=<rgba 1 0 0 1>}}
  User.IsClient {type=NiagaraBool}
  User.RibbonIDOffset {type=NiagaraInt32}
 ]
}
