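# Niagara burst effect played when a Keys and Locks vault opens, with a User.Color parameter and a shared system color tinting its emitters.
# Expanding distortion and shockwave flare meshes animate dynamic material parameters while a short-lived light fades out.
# Sprite flares ride a torus whose radius grows by curve, forming a radial ring.
# Center sparkles fly outward through curl noise and drag, with alpha scaled by Fortnite time of day.
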NiagaraSystem /KeysAndLocks/Items/KeysAndLocks/FX/NS_KeysLocks_VaultOpen
dependencies {
 usedDataInterfaces [
  &NiagaraDataInterfaceCurve
  &NiagaraDataInterfaceVectorCurve
  &NiagaraDataInterfaceColorCurve
  &NiagaraDataInterfaceVector2DCurve
 ]
 usedDynamicInputs []
 usedModules [
  &SystemState
  &self:SystemSpawnScript.SystemScriptSource.SystemScriptGraph.NiagaraNodeAssignment_0.SetVariables_#1
 ]
}
emitters {
 Distortion {
  emitterUpdateScript [
   &EmitterState(Life Cycle Mode=::NewEnumerator1, Loop Behavior=::NewEnumerator1, Loop Delay=0.025, Loop Duration=1, UseLoopDelay=true)
   &SpawnBurst_Instantaneous(Use Spawn Probability=false, UseLoopCountLimit=false)
  ]
  particleSpawnScript [
   /Niagara/Modules/Spawn/Initialization/InitializeParticle(
    Lifetime=0.35
    Mesh Scale=<5 5 5>
    Module.Write Color=true
    Module.Write Lifetime=true
    Module.Write Mass=true
    Module.Write Material Random=true
    Module.Write Position=false
    Module.Write Scale=true
    Module.Write SpriteRotation=false
    Module.Write SpriteSize=false
   )
   &Color(Color=$System.SysColor)
  ]
  particleUpdateScript [
   &ParticleState()
   /Niagara/Modules/Update/Orientation/V2/OrientMeshToVector(Look At Direction:Vector3f=<0 1 0>, Override Quat=false, UseOptionalSideAxis=false, UseOptionalUpAxis=false)
   &ScaleMeshSize(Scale Factor=&VectorFromFloat(Value=&FloatFromCurve(
      FloatCurve=DataInterface() {propertyValues {Curve {keys [
          {arriveTangent=6.638, interpMode=RCIM_Cubic, time=0, value=0}
          {arriveTangent=1, interpMode=RCIM_Cubic, leaveTangent=1, time=0.2381, value=0.8455}
          <at 1 1>
         ]}}}
      Scale Curve=0.3
     )))
   &DynamicMaterialParameters(
    Index 0 Param 1=&FloatFromCurve(FloatCurve=DataInterface() {propertyValues {Curve {keys [
         {arriveTangent=0.0451, interpMode=RCIM_Cubic, leaveTangent=0.0451, tangentMode=RCTM_User, time=0, value=1}
         {arriveTangent=-0.1426, interpMode=RCIM_Cubic, leaveTangent=-0.1426, tangentMode=RCTM_User, time=0.4, value=1.03}
         {arriveTangent=-0.0405, interpMode=RCIM_Cubic, leaveTangent=-0.0405, tangentMode=RCTM_User, time=1, value=1}
        ]}}})
    Index 0 Param 2=&FloatFromCurve(FloatCurve=DataInterface() {propertyValues {Curve {keys [<at 0 1.25 RCIM_Cubic>, {arriveTangent=-3.093, leaveTangent=-3.093, time=1, value=0}]}}})
    Index 0 Param 3=&FloatFromCurve(FloatCurve=DataInterface() {propertyValues {Curve {keys [{arriveTangent=0.6796, interpMode=RCIM_Cubic, time=0, value=-0.5}, <at 1 0>]}}})
    Index 0 Param 4=&FloatFromCurve(FloatCurve=DataInterface() {propertyValues {Curve {keys [{arriveTangent=0.2861, interpMode=RCIM_Cubic, time=0, value=0}, <at 1 0.12>]}}})
    Module.Param0WriteEnabled=true
    Module.Param1WriteEnabled=true
    Module.Param2WriteEnabled=true
    Module.Param3WriteEnabled=true
   )
   &ScaleColor(
    Scale Alpha=&FloatFromCurve(
     FloatCurve=DataInterface() {propertyValues {Curve {keys [<at 0 1 RCIM_Cubic>, {arriveTangent=-2.3476, leaveTangent=-2.3476, time=1, value=0}]}}}
     Scale Curve=0.1
    )
    Scale RGB=&VectorFromFloat(Value=&FloatFromCurve(
      FloatCurve=DataInterface() {propertyValues {Curve {keys [{arriveTangent=-23.8521, interpMode=RCIM_Cubic, time=0, value=10}, <at 1 1>]}}}
      Scale Curve=0.1
     ))
    ScaleA=true
    ScaleRGB=true
   )
   &CameraOffset(Camera Offset Amount=55)
  ]
  rendererClasses [&NiagaraMeshRendererProperties]
  renderers [{rendererClass=&NiagaraMeshRendererProperties, rendererIndex=0}]
 }
 LightRenderer {
  emitterUpdateScript [
   &EmitterState(Life Cycle Mode=::NewEnumerator1, Loop Behavior=::NewEnumerator1, Loop Duration=2, UseLoopDelay=false)
   &SpawnBurst_Instantaneous(Use Spawn Probability=false, UseLoopCountLimit=false)
  ]
  particleSpawnScript [
   /Niagara/Modules/Spawn/Initialization/V2/InitializeParticle(
    Color=$System.SysColor
    Lifetime=0.5
    Sprite Size Mode=::NewEnumerator1
    Uniform Sprite Size Max=100
    Uniform Sprite Size Min=55
    UsePositionOffset=false
    Write Material Random=true
   )
   &SphereLocation(
    MaskSpawnGroup=false
    Override Seed=false
    Sphere Radius=2
    Surface Only=false
    UseNonUniformScale=false
    UseOffset=true
    UseSphereOrientation=false
   )
   /SetVariables(Particles.LightRadius=75, Particles.LightScale:Vector3f=<1 1 1>) {isSetParameters=true}
  ]
  particleUpdateScript [
   &ParticleState()
   &ScaleColor(
    Scale Alpha=&FloatFromCurve(FloatCurve=DataInterface() {propertyValues {Curve {keys [<at 0.2 2>,<at 0.7 0.5>,<at 1 0>]}}}, Scale Curve=0.1)
    ScaleA=true
    ScaleRGB=true
   )
   &CameraOffset(Camera Offset Amount=55)
  ]
  rendererClasses [&NiagaraLightRendererProperties]
  renderers [{rendererClass=&NiagaraLightRendererProperties, rendererIndex=0}]
 }
 Shockwave_Flare {
  emitterUpdateScript [
   &EmitterState(Life Cycle Mode=::NewEnumerator1, Loop Behavior=::NewEnumerator1, Loop Duration=1, UseLoopDelay=true)
   &SpawnBurst_Instantaneous(Use Spawn Probability=false, UseLoopCountLimit=false)
  ]
  particleSpawnScript [
   /Niagara/Modules/Spawn/Initialization/InitializeParticle(
    Color=$System.SysColor
    Lifetime=0.5
    Mesh Scale=<5 5 5>
    Module.Write Color=true
    Module.Write Lifetime=true
    Module.Write Mass=true
    Module.Write Material Random=true
    Module.Write Position=true
    Module.Write Scale=true
    Module.Write SpriteRotation=false
    Module.Write SpriteSize=false
   )
   &self:Position() {module=Position}
   &Color(Color=$System.SysColor) {module=Color001}
  ]
  particleUpdateScript [
   &ParticleState()
   &OrientMeshToVector(Look At Direction:Vector3f=<1 0 0>, Override Quat=false, UseOptionalSideAxis=false, UseOptionalUpAxis=false)
   &DynamicMaterialParameters(
    Index 0 Param 1=&FloatFromCurve(FloatCurve=DataInterface() {propertyValues {Curve {keys [
         {arriveTangent=4.1785, interpMode=RCIM_Cubic, leaveTangent=4.1785, tangentMode=RCTM_User, time=0, value=0}
         {arriveTangent=0.9995, interpMode=RCIM_Cubic, leaveTangent=0.9995, tangentMode=RCTM_User, time=0.2223, value=0.6825}
         {arriveTangent=-0.0405, interpMode=RCIM_Cubic, leaveTangent=-0.0405, tangentMode=RCTM_User, time=1, value=1}
        ]}}})
    Index 0 Param 2=&FloatFromCurve(FloatCurve=DataInterface() {propertyValues {Curve {keys [<at 0 15 RCIM_Cubic>,<at 1 0>]}}})
    Index 0 Param 3=&FloatFromCurve(FloatCurve=DataInterface() {propertyValues {Curve {keys [<at 0 -0.8>,<at 1 -0.8>]}}})
    Index 0 Param 4=&FloatFromCurve(FloatCurve=DataInterface() {propertyValues {Curve {keys [<at 0 0>,<at 1 -1>]}}})
    Module.Param0WriteEnabled=true
    Module.Param1WriteEnabled=true
    Module.Param2WriteEnabled=true
    Module.Param3WriteEnabled=true
   )
   &ScaleMeshSize(Scale Factor=&VectorFromFloat(Value=&FloatFromCurve(
      FloatCurve=DataInterface() {propertyValues {Curve {keys [
          {arriveTangent=8.8797, interpMode=RCIM_Cubic, time=-0.0014, value=0}
          {arriveTangent=2.0056, interpMode=RCIM_Cubic, leaveTangent=2.0056, time=0.2495, value=1.3042}
          {interpMode=RCIM_Cubic, leaveTangent=0.4549, time=1, value=2.0084}
         ]}}}
      Scale Curve=0.2
     )))
   &ScaleColor(
    Scale Alpha=&FloatFromCurve(
     FloatCurve=DataInterface() {propertyValues {Curve {keys [
         <at -0.0043 2.2383 RCIM_Cubic>
         {arriveTangent=-4.9832, interpMode=RCIM_Cubic, leaveTangent=-4.9832, time=0.0972, value=1}
         {arriveTangent=-1.1077, interpMode=RCIM_Cubic, leaveTangent=-1.1077, time=0.3875, value=0.2861}
         <at 1 0>
        ]}}}
     Scale Curve=0.25
    )
    Scale RGB=&VectorFromCurve(VectorCurve=DataInterface() {propertyValues {
       XCurve {keys [<at 0 15 RCIM_Cubic>, {arriveTangent=-14.5, interpMode=RCIM_Cubic, leaveTangent=-14.5, time=0.1, value=5}, <at 1 0.5 RCIM_Cubic>]}
       YCurve {keys [<at 0 15 RCIM_Cubic>, {arriveTangent=-14.5, interpMode=RCIM_Cubic, leaveTangent=-14.5, time=0.1, value=5}, <at 1 0.5 RCIM_Cubic>]}
       ZCurve {keys [<at 0 15 RCIM_Cubic>, {arriveTangent=-14.5, interpMode=RCIM_Cubic, leaveTangent=-14.5, time=0.1, value=5}, <at 1 0.5 RCIM_Cubic>]}
      }})
    ScaleA=true
    ScaleRGB=true
   )
   &CameraOffset(Camera Offset Amount=125)
  ]
  rendererClasses [&NiagaraMeshRendererProperties]
  renderers [{rendererClass=&NiagaraMeshRendererProperties, rendererIndex=0}]
 }
 radial_flares001 {
  emitterUpdateScript [
   &EmitterState(Inactive Response=::NewEnumerator1, Life Cycle Mode=::NewEnumerator1, Loop Behavior=::NewEnumerator1, Loop Duration=0.3, UseLoopDelay=false)
   &SpawnBurst_Instantaneous(Spawn Count=35, Use Spawn Probability=false, UseLoopCountLimit=false)
  ]
  particleSpawnScript [/Niagara/Modules/Spawn/Initialization/V2/InitializeParticle(Color=$System.SysColor, Sprite Size=<1 15>, Sprite Size Mode=::NewEnumerator4, UsePositionOffset=false, Write Material Random=true)]
  particleUpdateScript [
   &ParticleState()
   &TorusLocation(
    Align Torus With Axis=true
    Handle Radius=5
    Large Radius:NiagaraFloat=&FloatFromCurve(
     FloatCurve=DataInterface() {propertyValues {Curve {keys [
         {arriveTangent=9.0725, interpMode=RCIM_Cubic, leaveTangent=9.0725, leaveTangentWeight=0.3879, tangentMode=RCTM_User, time=0, value=0}
         {
          arriveTangent=7.6747
          arriveTangentWeight=0.1
          interpMode=RCIM_Cubic
          leaveTangent=7.6747
          leaveTangentWeight=0.2333
          tangentMode=RCTM_User
          time=0.2
          value=2.125
         }
         {
          arriveTangent=7.6747
          arriveTangentWeight=0.0333
          interpMode=RCIM_Cubic
          leaveTangent=7.6747
          leaveTangentWeight=0.2333
          tangentMode=RCTM_User
          time=0.3
          value=2.75
         }
         {arriveTangent=-0.0405, arriveTangentWeight=0.2335, interpMode=RCIM_Cubic, leaveTangent=-0.0405, tangentMode=RCTM_User, time=1, value=4.375}
        ]}}}
     Scale Curve=&RandomRangeFloat(Minimum=5, Maximum=12)
    )
    MaskSpawnGroup=false
    Normalized Angle Around Torus Axis=1
    Torus Axis=<0 1 0>
    Torus Distribution Mode:ENiagaraTorusDistributionMode=::NewEnumerator1
    U Position:NiagaraFloat=&Add_Float(
     A=&ReturnNormalizedExecIndex()
     B=&ModuloFloat(A=&Multiply_Float(A=$Particles.Age, B=0.001), B=1)
    )
   )
   &self:ScratchModule_05() {module=ScratchModule_05}
   &ScaleColor(
    Scale Alpha=&FloatFromCurve(
     FloatCurve=DataInterface() {propertyValues {Curve {keys [
         {arriveTangent=-0.2933, interpMode=RCIM_Cubic, leaveTangent=-0.2933, tangentMode=RCTM_User, time=0, value=0}
         {arriveTangent=-3.5801, interpMode=RCIM_Cubic, leaveTangent=-3.5801, tangentMode=RCTM_User, time=0.1, value=1}
         {arriveTangent=0.0301, interpMode=RCIM_Cubic, leaveTangent=0.0301, tangentMode=RCTM_User, time=0.9, value=0.1}
        ]}}}
     Scale Curve=0.075
    )
    ScaleA=true
    ScaleRGB=false
   ) {module=ScaleColor001}
   &ScaleSpriteSize(
    Uniform Curve Scale=1.25
    Uniform Curve Sprite Scale=DataInterface() {propertyValues {Curve {keys [
        {arriveTangent=-0.2933, interpMode=RCIM_Cubic, leaveTangent=-0.2933, leaveTangentWeight=0.4283, tangentMode=RCTM_User, time=0, value=0}
        {
         arriveTangent=-3.5801
         arriveTangentWeight=0.1274
         interpMode=RCIM_Cubic
         leaveTangent=-3.5801
         leaveTangentWeight=0.7778
         tangentMode=RCTM_User
         time=0.1408
         value=0.9847
        }
        {arriveTangent=0.0301, arriveTangentWeight=0.2866, interpMode=RCIM_Cubic, leaveTangent=0.0301, tangentMode=RCTM_User, time=1, value=0}
       ]}}}
   )
   /SetVariables(Particles.SpriteSize_B:Vector2f=<15 55>) {isSetParameters=true}
  ]
  rendererClasses [&NiagaraSpriteRendererProperties]
  renderers [{rendererClass=&NiagaraSpriteRendererProperties}]
 }
 sparkle_center {
  emitterUpdateScript [
   &EmitterState(Inactive Response=::NewEnumerator1, Life Cycle Mode=::NewEnumerator1, Loop Behavior=::NewEnumerator1, Loop Duration=2, UseLoopDelay=false)
   &SpawnBurst_Instantaneous(Spawn Count=88, Use Spawn Probability=false, UseLoopCountLimit=false)
  ]
  particleSpawnScript [
   &InitializeParticle(
    Color=$System.SysColor
    Lifetime=&RandomRangeFloat(Minimum=0.3, Maximum=1.2)
    Sprite Rotation Mode=::NewEnumerator2
    Sprite Size Mode=::NewEnumerator1
    Uniform Sprite Size Max=6
    Uniform Sprite Size Min=3.5
    UsePositionOffset=false
    Write Material Random=true
   )
   /SetVariables(Particles.RandomCore:NiagaraFloat=&FloatFromCurve(
     FloatCurve=DataInterface() {propertyValues {Curve {keys [
         {arriveTangent=0.0559, interpMode=RCIM_Cubic, leaveTangent=0.0559, tangentMode=RCTM_User, time=0, value=0.2}
         {arriveTangent=0.1895, interpMode=RCIM_Cubic, leaveTangent=0.1895, tangentMode=RCTM_User, time=0.9, value=0.25}
         {arriveTangent=1.1808, interpMode=RCIM_Cubic, leaveTangent=1.1808, tangentMode=RCTM_User, time=1, value=0.3}
        ]}}}
     CurveIndex=&RandomRangeFloat(Minimum=0.1, Maximum=1)
    )) {isSetParameters=true}
   &SphereLocation(
    MaskSpawnGroup=false
    Override Seed=false
    Sphere Radius=20
    Surface Only=false
    UseNonUniformScale=false
    UseOffset=false
    UseSphereOrientation=false
   )
   &AddVelocityFromPoint(Module.Use Falloff Distance=false, Velocity Strength=222)
  ]
  particleUpdateScript [
   &ParticleState()
   &CurlNoiseForce(Noise Frequency=125, Noise Strength=&Multiply_Float(A=555, B=1), Pan Noise=true, Pan Noise Field=<0.5 0.5 0>)
   &ScaleSpriteSize(Scale Factor=&Vector2DFromFloat(Value=&FloatFromCurve(
      FloatCurve=DataInterface() {propertyValues {Curve {keys [
          <at 0 1 RCIM_Cubic>
          {arriveTangent=-1, interpMode=RCIM_Cubic, leaveTangent=-1, time=0.6, value=1.25}
          {arriveTangent=-0.1423, interpMode=RCIM_Cubic, leaveTangent=-0.1423, tangentMode=RCTM_User, time=1, value=0}
         ]}}}
      Scale Curve=$Particles.RandomCore
     )))
   &ScaleSpriteSize(
    Scale Factor=&Multiply_Vector2D(A=$Particles.SpriteSize, B=<0.15 0.15>)
    Scale Sprite Size Mode=::NewEnumerator2
   ) {module=ScaleSpriteSize001}
   &ScaleColor(
    Scale Alpha=&Multiply_Float(
     A=&FloatFromCurve(FloatCurve=DataInterface() {propertyValues {Curve {keys [<at 0.8 1>,<at 1 0>]}}})
     B=&FNTimeOfDayScalar(Night Scale=0.3, Morning Scale=0.1, Evening Scale=0.2)
    )
    Scale RGB=&VectorFromFloat(Value=&FloatFromCurve(FloatCurve=DataInterface() {propertyValues {Curve {keys [<at 0 1>,<at 0.1 2>,<at 0.2 1>]}}}))
    ScaleA=true
    ScaleRGB=true
   )
   &self:ScratchModule_08() {module=ScratchModule_08}
   &Drag(Use Linear Drag=true, Use Rotational Drag=false)
   &SolveForcesAndVelocity(Clamp Velocity=false, Limit Acceleration=false)
  ]
  rendererClasses [&NiagaraSpriteRendererProperties]
  renderers [{rendererClass=&NiagaraSpriteRendererProperties}]
 }
}
summary {
 systemName=NS_KeysLocks_VaultOpen
 emitters [
  {emitterName=LightRenderer, rendererClasses [&NiagaraLightRendererProperties]}
  {emitterName=Shockwave_Flare, rendererClasses [&NiagaraMeshRendererProperties]}
  {emitterName=Distortion, rendererClasses [&NiagaraMeshRendererProperties]}
  {emitterName=radial_flares001, rendererClasses [&NiagaraSpriteRendererProperties]}
  {emitterName=sparkle_center, rendererClasses [&NiagaraSpriteRendererProperties]}
 ]
 userVariables [User.Color {type=LinearColor, description="", defaultValue {struct=LinearColor, value=<rgba 0 1 0 1>}}]
}
