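; Keys and Locks chest Blueprint, derived from the tiered Athena chest parent, that manages the chest's openable cosmetics.
; HideChestOpenableCosmetics runs once to hide smoke and lights, destroy the aura, call opened audio, and shut down indicators, wiggle materials, mobile selection visuals and custom depth.
; The construction script adds a TieredChestAthena material tag so level-placed chests receive the correct chrome material.
; BeginPlay, overlap and tick events add no logic of their own.
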(event EventBeginPlay
  (|CallParentFunction) ; This node is disabled and will not be called.\nDrag off pins to build functionality.
)

(event Collision|EventActorBeginOverlap
  (params (OtherActor :type Actor%20Object%20Reference)))

(event EventTick
  (params (DeltaSeconds :type Float%20%28single-precision%29)))

(event EventHideChestOpenableCosmetics
  (Utilities|FlowControl|DoOnce false
    (:Completed
      (Rendering|SetVisibility (Variables|Default|GetCustom%20Smoke%20Sheet) false false)
      (Components|DestroyComponent (Variables|Default|GetCustom%20Chest%20Aura))
      (CallFunction|SetLightVisibility false)
      (CallFunction|AudioonOpened)
      (CallFunction|StopAudiovisualIndicator)
      (CallFunction|CleanupWiggleMIDs)
      (Components|Timeline|Stop (Variables|Tiered_Chest_Athena_Parent|GetMobileSelectedTL))
      (CallFunction|SetVisibleMobileInteractIcon false)
      (Rendering|SetRenderCustomDepth (Variables|Debug|GetStaticMeshComponent) false))))

(fn ConstructionScript
  ;[0.32 0.9402 1] Make sure actors that are already placed in levels get this tag so that they get the right chrome material added to them.
  (|CallParentFunction)
  (Utilities|Array|AddUnique (Variables|Actor|GetTags) "Lackluster.Material.TieredChestAthena"))
