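# M_KeysAndLocks_Rays is an additive, unlit, two-sided material for glowing rays on key and lock item meshes.
# An EmissionColor parameter tints the glow, and Opacity scales both the color and the ray mask.
# The mask is a steep power falloff along inverted V, zeroed where U is below LockWipe so the parameter wipes rays away.
# A time-driven sine, remapped between PulseMin and PulseMax, pulses the mask at PulseSpeed.
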Material /KeysAndLocks/Items/KeysAndLocks/Meshes/Materials/M_KeysAndLocks_Rays
graph {
 properties {blendMode=BLEND_Additive, pixelDepthOffsetMode=PDOM_Legacy, shadingModel=MSM_Unlit, twoSided=true, usedShadingModels=Unlit}
 nodes [
  n1 = ScalarParameter() {defaultValue=1, parameterName=Opacity, sliderMax=1}
  Multiply(A=VectorParameter() {parameterName=EmissionColor, defaultValue=<rgba 1 0 0 1>}.RGB, B=@n1)
  n2 = ScalarParameter() {parameterName=LockWipe, sliderMax=1, sliderMin=1}
  n3 = MaterialFunctionCall(Float2=TextureCoordinate()) {materialFunction=&BreakOutFloat2Components}
  n4 = Constant()
  n5 = Power(Base=OneMinus(None=@n3.G)) {constExponent=8}
  n6 = Sine(None=Multiply(A=Time(), B=ScalarParameter() {defaultValue=0.2, parameterName=PulseSpeed, sliderMax=5}))
  n7 = Constant() {r=-1}
  n8 = Constant() {r=1}
  n9 = ScalarParameter() {defaultValue=0.25, parameterName=PulseMin, sliderMax=1}
  n10 = ScalarParameter() {defaultValue=1, parameterName=PulseMax, sliderMax=1}
  Multiply(
   A=Multiply(
    A=If(A=@n2, B=@n3.R, "A > B"=@n4, "A < B"=@n5)
    B=MaterialFunctionCall(Input=@n6, Input Low=@n7, Input High=@n8, Target Low=@n9, Target High=@n10) {materialFunction=&RemapValueRange}.Result
   )
   B=@n1
  )
 ]
}
